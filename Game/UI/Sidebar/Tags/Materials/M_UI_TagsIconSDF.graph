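# Translucent UI material for sidebar tag icons stored as signed distance field textures, defaulting to a voice chat lock icon.
# The Icon SDF texture is sampled with normalized UI UVs, and its red channel feeds MF_UI_AntiAliasing with a Threshold parameter for smooth edges.
# A Color parameter is exposed alongside the texture in the Icon group, making texture, color and threshold adjustable per instance.
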Material /Game/UI/Sidebar/Tags/Materials/M_UI_TagsIconSDF
graph {
 properties {
  bUsedWithStaticMesh=false
  blendMode=BLEND_Translucent
  materialDomain=MD_UI
  pixelDepthOffsetMode=PDOM_Legacy
  usedShadingModels=Default Lit
 }
 nodes [
  n1 = TextureSampleParameter2D(UVs=MaterialFunctionCall() {materialFunction=&GetUserInterfaceUV}.Normalized UV) {
   group=0 - Icon
   parameterName=Icon SDF
   samplerType=SAMPLERTYPE_LinearColor
   sortPriority=0
   texture=/Game/UI/Sidebar/Textures/T_UI_Icon_Sidebar_VoiceChatLock_SDF
  }
  MaterialFunctionCall(DistanceField=@n1.R, Threshold=ScalarParameter() {defaultValue=0.25, group=1 - Distance Field, parameterName=Threshold, sortPriority=1}) {materialFunction=&MF_UI_AntiAliasing}
  VectorParameter() {group=0 - Icon, parameterName=Color, sortPriority=1, defaultValue=<rgba 1 1 1 1>}
 ]
}
