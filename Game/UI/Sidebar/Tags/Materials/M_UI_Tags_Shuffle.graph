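# Translucent UI material for the sidebar tags' Shuffle icon, animating a front and a back arrow built from one channel-packed texture.
# Each arrow has its own offset and scale, with the back one flipped vertically.
# The blue channel's gradient combines with delayed, looping time through Frac and SmoothStep to drive each arrow's animation.
# A Max merges the masked arrows, one cut by a smoothstepped mask; green supplies a remapped front-arrow SDF.
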Material /Game/UI/Sidebar/Tags/Materials/M_UI_Tags_Shuffle
graph {
 properties {
  bUsedWithStaticMesh=false
  blendMode=BLEND_Translucent
  materialDomain=MD_UI
  pixelDepthOffsetMode=PDOM_Legacy
  usedShadingModels=Default Lit
 }
 nodes [
  n1 = NamedRerouteUsage()
  n2 = NamedRerouteUsage()
  n3 = Multiply(A=NamedRerouteUsage(), B=NamedRerouteUsage())
  Max(
   A=Multiply(A=NamedRerouteUsage(), B=NamedRerouteUsage())
   B=Subtract(A=Multiply(A=@n1, B=@n2) {constA=1}, B=SmoothStep(Value=@n3) {constMax=0.2})
  )
  n4 = AppendVector(
   A=ScalarParameter() {group=Arrow Front, parameterName=Front Arrow Offset X, sortPriority=34}
   B=ScalarParameter() {group=Arrow Front, parameterName=Front Arrow Offset Y, sortPriority=35}
  )
  n5 = TextureCoordinate()
  n6 = ScalarParameter() {defaultValue=1, group=Arrow Front, parameterName=Front Arrow Scale X}
  n7 = ScalarParameter() {defaultValue=1, group=Arrow Front, parameterName=Front Arrow Scale Y, sortPriority=33}
  n8 = TextureSampleParameter2D(UVs=MaterialFunctionCall(UVs=Subtract(A=@n5, B=@n4), Texture Scale=AppendVector(A=@n6, B=@n7)) {materialFunction=&ScaleUVsByCenter}.UVs) {
   group=Icon
   parameterName=Arrow Icon
   samplerType=SAMPLERTYPE_LinearColor
   sortPriority=33
   texture=/Game/UI/Sidebar/Tags/Textures/T_UI_Tags_Shuffle
  }
  NamedRerouteDeclaration(None=@n8.R) {name=Front Arrow, nodeColor=<rgba 0 0 0 1>}
  n9 = AppendVector(
   A=ScalarParameter() {group=Arrow Back, parameterName=Back Arrow Offset X, sortPriority=34}
   B=ScalarParameter() {group=Arrow Back, parameterName=Back Arrow Offset Y, sortPriority=35}
  )
  n10 = ScalarParameter() {defaultValue=1, group=Arrow Back, parameterName=Back Arrow Scale X}
  n11 = ScalarParameter() {defaultValue=1, group=Arrow Back, parameterName=Back Arrow Scale Y, sortPriority=33}
  n12 = MaterialFunctionCall(Float2=MaterialFunctionCall(UVs=Subtract(A=@n5, B=@n9), Texture Scale=AppendVector(A=@n10, B=@n11)) {materialFunction=&ScaleUVsByCenter}.UVs) {materialFunction=&BreakOutFloat2Components}
  n13 = TextureSampleParameter2D(UVs=MaterialFunctionCall(X=@n12.R, Y=OneMinus(None=@n12.G)) {materialFunction=&MakeFloat2}.Result) {
   group=Icon
   parameterName=Arrow Icon
   samplerType=SAMPLERTYPE_LinearColor
   sortPriority=33
   texture=&T_UI_Tags_Shuffle
  }
  NamedRerouteDeclaration(None=@n13.R) {name=Back Arrow, nodeColor=<rgba 0 0 0 1>}
  n14 = NamedRerouteUsage()
  n15 = Multiply(
   A=Subtract(A=Time(), B=ScalarParameter() {group=Animation, parameterName=Back Delay})
   B=ScalarParameter() {defaultValue=0.5, group=Animation, parameterName=Back Time Speed}
  )
  n16 = Subtract(
   A=Power(Base=OneMinus(None=@n14), Exp=ScalarParameter() {defaultValue=0.5, group=Animation, parameterName=Back Gradient Speed}) {constExponent=0.5}
   B=Multiply(A=Frac(None=@n15)) {constB=2}
  )
  NamedRerouteDeclaration(None=SmoothStep(
    Min=ScalarParameter() {defaultValue=0.5, group=Animation, parameterName=Back Fade Min, sliderMax=1}
    Max=ScalarParameter() {defaultValue=1, group=Animation, parameterName=Back Fade Max, sliderMax=1}
    Value=Frac(None=@n16)
   ) {constMin=0.5}) {name=Back Arrow Animation, nodeColor=<rgba 0 0 0 1>}
  n17 = Constant() {desc=Min, r=0.5}
  n18 = Subtract(A=@n8.G, B=@n17)
  n19 = Subtract(A=Constant() {desc=Max, r=1}, B=@n17)
  NamedRerouteDeclaration(None=Saturate(None=Divide(A=@n18, B=@n19))) {name=Front Arrow SDF, nodeColor=<rgba 0 0 0 1>}
  NamedRerouteDeclaration(None=@n8.B) {name=Front Arrow Time Gradient, nodeColor=<rgba 0 0 0 1>}
  NamedRerouteDeclaration(None=@n13.B) {name=Back Arrow Time Gradient, nodeColor=<rgba 0 0 0 1>}
  n20 = NamedRerouteUsage()
  n21 = Multiply(
   A=Subtract(A=Time(), B=ScalarParameter() {defaultValue=0.5, group=Animation, parameterName=Front Delay})
   B=ScalarParameter() {defaultValue=0.5, group=Animation, parameterName=Front Time Speed}
  )
  n22 = Subtract(
   A=Power(Base=OneMinus(None=@n20), Exp=ScalarParameter() {defaultValue=0.5, group=Animation, parameterName=Front Gradient Speed}) {constExponent=0.5}
   B=Multiply(A=Frac(None=@n21)) {constB=2}
  )
  NamedRerouteDeclaration(None=SmoothStep(
    Min=ScalarParameter() {defaultValue=0.5, group=Animation, parameterName=Front Fade Min, sliderMax=1}
    Max=ScalarParameter() {defaultValue=1, group=Animation, parameterName=Front Fade Max, sliderMax=1}
    Value=Frac(None=@n22)
   ) {constMin=0.5}) {name=Front Arrow Animation, nodeColor=<rgba 0 0 0 1>}
  VectorParameter() {group=Icon, parameterName=Arrow Color, sortPriority=35, defaultValue=<rgba 1 1 1 1>}
 ]
}
