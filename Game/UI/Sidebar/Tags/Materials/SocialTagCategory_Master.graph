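# Translucent UI material for sidebar social tag category icons, with a configurable gradient, icon color, and hover and disabled states.
# The TagCategoryIcon texture packs back and front layers into R and G, each center-scaled over a different range by IsHovered for a layered hover effect.
# A rotatable linear gradient and the icon color each blend to gray by IsDisabled, brightened by DisabledHover, then composite through an icon mask.
# A Max combines two mask values.
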Material /Game/UI/Sidebar/Tags/Materials/SocialTagCategory_Master
graph {
 properties {
  bUsedWithStaticMesh=false
  blendMode=BLEND_Translucent
  materialDomain=MD_UI
  pixelDepthOffsetMode=PDOM_Legacy
  usedShadingModels=Default Lit
 }
 nodes [
  n1 = VectorParameter() {group=Gradient, parameterName=Gradient Start Color, sortPriority=35, defaultValue=<rgba 0 0 0 1>}
  n2 = VectorParameter() {group=Gradient, parameterName=Gradient End Color, sortPriority=36, defaultValue=<rgba 1 1 1 1>}
  n3 = MaterialFunctionCall(
   GradientStart=ScalarParameter() {group=Gradient, parameterName=Gradient Start, sliderMax=1, sortPriority=37}
   GradientEnd=ScalarParameter() {defaultValue=1, group=Gradient, parameterName=Gradient End, sliderMax=1, sortPriority=38}
   RotationAngle=MaterialFunctionCall!(Degrees=ScalarParameter() {group=Gradient, parameterName=Rotation Angle, sliderMax=360, sortPriority=39}) {materialFunction=/Game/UI/Sidebar/Materials/MF_UI_ConvertDegreesToUnitInterval}.Unit Interval
   GradientOffset=ScalarParameter() {group=Gradient, parameterName=Gradient Offset, sliderMax=1, sliderMin=-1, sortPriority=40}
  ) {materialFunction=&MF_UI_LinearGradient}
  n4 = LinearInterpolate(
   A=Constant4Vector() {desc=Gradient Color Disabled, constant=<rgba 0.1146 0.1146 0.1146 0.3>}.RGBA
   B=Constant4Vector() {desc=Gradient Color Disabled Hovered, constant=<rgba 0.2292 0.2292 0.2292 0.2>}.RGBA
   Alpha=ScalarParameter() {group=Hover Animation, parameterName=DisabledHover, sliderMax=1, sortPriority=33}
  ) {constA=1, constB=1.2}
  n5 = LinearInterpolate(
   A=Constant4Vector() {desc=Icon Color Disabled, constant=<rgba 0.901 0.901 0.901 0.6>}.RGBA
   B=Constant4Vector() {desc=Icon Color Disabled Hovered, constant=<rgba 1 1 1 1>}.RGBA
   Alpha=ScalarParameter() {group=Hover Animation, parameterName=DisabledHover, sliderMax=1, sortPriority=33}
  ) {constA=1, constB=1.2}
  LinearInterpolate(
   A=LinearInterpolate(
    A=LinearInterpolate(A=@n1.RGB, B=@n2.RGB, Alpha=@n3.Result)
    B=ComponentMask(None=@n4) {b=true, g=true, r=true}
    Alpha=ScalarParameter() {group=Hover Animation, parameterName=IsDisabled, sliderMax=1, sortPriority=33}
   ) {constA=1, constB=1.2}
   B=LinearInterpolate(
    A=VectorParameter() {group=Icon, parameterName=Icon Color, sortPriority=34, defaultValue=<rgba 1 1 1 1>}.RGB
    B=ComponentMask(None=@n5) {b=true, g=true, r=true}
    Alpha=ScalarParameter() {group=Hover Animation, parameterName=IsDisabled, sliderMax=1, sortPriority=33}
   ) {constA=1, constB=1.2}
   Alpha=NamedRerouteUsage()
  )
  Max(A=NamedRerouteUsage(), B=NamedRerouteUsage())
  n6 = Constant() {r=0.9}
  n7 = Constant() {r=1}
  n8 = ScalarParameter() {group=Hover Animation, parameterName=IsHovered, sliderMax=1, sliderMin=-1, sortPriority=33}
  n9 = TextureCoordinate()
  n10 = TextureSampleParameter2D(UVs=MaterialFunctionCall(UVs=@n9, Texture Scale=LinearInterpolate(A=@n6, B=@n7, Alpha=@n8) {constA=1, constB=1.2}) {materialFunction=&ScaleUVsByCenter}.UVs) {
   group=Icon
   parameterName=TagCategoryIcon
   samplerType=SAMPLERTYPE_LinearColor
   sortPriority=33
   texture=/Game/UI/Sidebar/Tags/Textures/TagCategoryIcon_GameMode
  }
  NamedRerouteDeclaration(None=@n10.R) {name=TagCategoryIcon_Back, nodeColor=<rgba 0 0 0 1>}
  n11 = Constant() {r=0.9}
  n12 = Constant() {r=1.1}
  n13 = TextureSampleParameter2D(UVs=MaterialFunctionCall(UVs=@n9, Texture Scale=LinearInterpolate(A=@n11, B=@n12, Alpha=@n8) {constA=1, constB=1.2}) {materialFunction=&ScaleUVsByCenter}.UVs) {
   group=Icon
   parameterName=TagCategoryIcon
   samplerType=SAMPLERTYPE_LinearColor
   sortPriority=33
   texture=&TagCategoryIcon_GameMode
  }
  NamedRerouteDeclaration(None=@n13.G) {name=TagCategoryIcon_Front, nodeColor=<rgba 0 0 0 1>}
 ]
}
