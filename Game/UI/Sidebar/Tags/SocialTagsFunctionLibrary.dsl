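; Function library for sidebar social tag widgets, providing category lookup and hover color interpolation.
; Category names, stripped of spaces, select data table rows, falling back to the AddTag row.
; Caller-supplied alphas lerp border gradient, stroke and inner/outer colors through dynamic material parameters, plus tag text and check mark colors.
; Border functions clear the caller's timer when interpolation completes, with gradient and stroke variants also checking widget hover state.
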(fn ClearLerpColorTimer
  (params (ColorLerpAlpha :type Float%20%28double-precision%29) (ColorLerpTimeHandle :type Timer%20Handle%20Structure%20%28by%20ref%29))
  ;[0.1779 0.005004 1] If lerp value is 1, stop interpolation.
  (if (== ColorLerpAlpha 1.0)
    (Utilities|Time|ClearandInvalidateTimerbyHandle 0 ColorLerpTimeHandle)))

(fn GetSidebarSocialTagCategoryStructure
  (params (TagCategory :type String))
  ;[1 0.6898 0.154] Get SidebarSocialTagCategoryStructure row from SidebarSocialTagCategoryDataTable
  (bind _returnvalue (Utilities|GetDataTableRow "/Game/UI/Sidebar/Tags/SidebarSocialTagCategoryDataTable.SidebarSocialTagCategoryDataTable" (Utilities|String|StringToName (Utilities|String|Replace TagCategory " ")))
    (:then
      (LocalVariables|SetLocalSidebarSocialTagCategoryStructure _returnvalue))
    (:RowNotFound
      (bind _returnvalue_2 (Utilities|GetDataTableRow "/Game/UI/Sidebar/Tags/SidebarSocialTagCategoryDataTable.SidebarSocialTagCategoryDataTable" "AddTag"))
      (LocalVariables|SetLocalSidebarSocialTagCategoryStructure _returnvalue_2)))
  (return (LocalVariables|GetLocalSidebarSocialTagCategoryStructure)))

(fn InterpolateMaterialColor
  (params (Material :type Material%20Instance%20Dynamic%20Object%20Reference) (ColorParameterName :type Name) (CurrentColor :type Linear%20Color%20Structure) (TargetColor :type Linear%20Color%20Structure) (ColorLerpAlpha :type Float%20%28double-precision%29))
  ;[0.323 0 1] Interpolate color using the input color and alpha values
  (Rendering|Material|SetVectorParameterValue Material ColorParameterName (Math|Color|Lerp(LinearColor) CurrentColor TargetColor ColorLerpAlpha) :!class /Script/Engine.MaterialInstanceDynamic))

(fn InterpolateTagBorderColors
  (params (TagBorderMaterial :type Material%20Instance%20Dynamic%20Object%20Reference) (NameInnerColor :type Name) (CurrentInnerColor :type Linear%20Color%20Structure) (TargetInnerColor :type Linear%20Color%20Structure) (NameOuterColor :type Name) (CurrentOuterColor :type Linear%20Color%20Structure) (TargetOuterColor :type Linear%20Color%20Structure) (HoveredBorderColorLerpAlpha :type Float%20%28double-precision%29) (ColorLerpTimeHandle :type Timer%20Handle%20Structure%20%28by%20ref%29) (WorldContext :type Object%20Reference))
  ;[0.2359 0.08067 1] Interpolate tag border inner and outer colors to hovered/unhovered color values using the lerp alpha that is animated in the timeline.
  (Default|InterpolateMaterialColor TagBorderMaterial NameInnerColor CurrentInnerColor TargetInnerColor HoveredBorderColorLerpAlpha WorldContext)
  (Default|InterpolateMaterialColor TagBorderMaterial NameOuterColor CurrentOuterColor TargetOuterColor HoveredBorderColorLerpAlpha WorldContext)
  (CallFunction|ClearLerpColorTimer HoveredBorderColorLerpAlpha ColorLerpTimeHandle WorldContext))

(fn InterpolateTagBorderGradientColor
  (params (TagBorder :type Border%20Object%20Reference) (TagInnerColorGradient_1 :type Name) (TagInnerColorGradient_2 :type Name) (SidebarSocialTagCategoryStructure :type Sidebar%20Social%20Tag%20Category%20Structure%20Structure%20%28by%20ref%29) (TagBorderInnerColorHoveredLerpValue :type Float%20%28double-precision%29) (TagBorderInnerColorHoveredLerpTimeHandle :type Timer%20Handle%20Structure%20%28by%20ref%29) (TagWidgetReference :type Widget%20Object%20Reference))
  (bind (_tagskewboxinnercolor_48_c74e0ab74864de4e29c3aeba2ecc7da4 _tagskewboxinnercolorhovered_42_7c212cfc4a6078eee7c986b14e3feb49) (Utilities|Struct|BreakSidebarSocialTagCategoryStructure SidebarSocialTagCategoryStructure))
  (Utilities|FlowControl|Sequence
    (:then_0
      ;[0.323 0 1] Interpolate inner gradient colors to hovered/unhovered color values using the lerp value that is animated in the timeline.
      (bind _returnvalue (Appearance|GetDynamicMaterial TagBorder :!class /Script/UMG.Border))
      (Utilities|IsValid _returnvalue
        (:Is%20Valid
          (Rendering|Material|SetVectorParameterValue _returnvalue TagInnerColorGradient_1 (Math|Color|Lerp(LinearColor) _tagskewboxinnercolor_48_c74e0ab74864de4e29c3aeba2ecc7da4 _tagskewboxinnercolorhovered_42_7c212cfc4a6078eee7c986b14e3feb49 TagBorderInnerColorHoveredLerpValue) :!class /Script/Engine.MaterialInstanceDynamic)
          (Rendering|Material|SetVectorParameterValue _returnvalue TagInnerColorGradient_2 (Math|Color|Lerp(LinearColor) :A _tagskewboxinnercolor_48_c74e0ab74864de4e29c3aeba2ecc7da4 :Alpha TagBorderInnerColorHoveredLerpValue) :!class /Script/Engine.MaterialInstanceDynamic))))
    (:then_1
      ;[0.1779 0.005004 1] If hovered and lerp value is 1, or if unhovered and lerp value is 0, stop interpolation.
      (if (or (and (== TagBorderInnerColorHoveredLerpValue 1.0) (Widget|IsHovered TagWidgetReference)) (and (not (Widget|IsHovered TagWidgetReference)) (== TagBorderInnerColorHoveredLerpValue 0)))
        (Utilities|Time|ClearandInvalidateTimerbyHandle 0 TagBorderInnerColorHoveredLerpTimeHandle)))))

(fn InterpolateTagBorderOuterStrokeColor
  (params (TagBorder :type Border%20Object%20Reference) (TagOuterStrokeColor :type Name) (SidebarSocialTagCategoryStructure :type Sidebar%20Social%20Tag%20Category%20Structure%20Structure%20%28by%20ref%29) (TagBorderInnerColorHoveredLerpValue :type Float%20%28double-precision%29) (TagBorderInnerColorHoveredLerpTimeHandle :type Timer%20Handle%20Structure%20%28by%20ref%29) (TagWidgetReference :type Widget%20Object%20Reference) (TagOuterStrokeColorUnhovered :type Linear%20Color%20Structure))
  (bind (_tagskewboxinnercolor_48_c74e0ab74864de4e29c3aeba2ecc7da4 _tagskewboxinnercolorhovered_42_7c212cfc4a6078eee7c986b14e3feb49 _tagskewboxinnercolorselected_45_e34ff8a8455fe97079f6b0a6a31a05ab _tagskewboxoutercolor_41_98ddef0940105128e6c274b11d3a3d5d) (Utilities|Struct|BreakSidebarSocialTagCategoryStructure SidebarSocialTagCategoryStructure))
  (Utilities|FlowControl|Sequence
    (:then_0
      ;[0.323 0 1] Interpolate inner gradient colors to hovered/unhovered color values using the lerp value that is animated in the timeline.
      (bind _returnvalue (Appearance|GetDynamicMaterial TagBorder :!class /Script/UMG.Border))
      (Utilities|IsValid _returnvalue
        (:Is%20Valid
          (Rendering|Material|SetVectorParameterValue _returnvalue TagOuterStrokeColor (Math|Color|Lerp(LinearColor) TagOuterStrokeColorUnhovered _tagskewboxoutercolor_41_98ddef0940105128e6c274b11d3a3d5d TagBorderInnerColorHoveredLerpValue) :!class /Script/Engine.MaterialInstanceDynamic))))
    (:then_1
      ;[0.1779 0.005004 1] If hovered and lerp value is 1, or if unhovered and lerp value is 0, stop interpolation.
      (if (or (and (== TagBorderInnerColorHoveredLerpValue 1.0) (Widget|IsHovered TagWidgetReference)) (and (not (Widget|IsHovered TagWidgetReference)) (== TagBorderInnerColorHoveredLerpValue 0)))
        (Utilities|Time|ClearandInvalidateTimerbyHandle 0 TagBorderInnerColorHoveredLerpTimeHandle)))))

(fn InterpolateTagContentColors
  (params (TagName :type Text%20Object%20Reference) (CheckMark :type Image%20Object%20Reference) (CurrentTagContentColor :type Linear%20Color%20Structure) (TargetTagContentColor :type Linear%20Color%20Structure) (HoveredTagContentLerpAlpha :type Float%20%28double-precision%29))
  ;[0.2554 0.07767 1] Interpolate tag text and check mark colors to hovered/unhovered color values using the lerp alpha that is animated in the timeline.
  (Utilities|IsValid TagName
    (:Is%20Valid
      (Appearance|SetColorandOpacity TagName (Utilities|Struct|MakeSlateColor (Math|Color|Lerp(LinearColor) CurrentTagContentColor TargetTagContentColor HoveredTagContentLerpAlpha)) :!class /Script/UMG.TextBlock)
      (Utilities|IsValid CheckMark
        (:Is%20Valid
          (Appearance|SetColorandOpacity CheckMark (Math|Color|Lerp(LinearColor) CurrentTagContentColor TargetTagContentColor HoveredTagContentLerpAlpha) :!class /Script/UMG.Image))))))
