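# Sidebar tab button widget for tag editing, subclassing FortSidebarCollectionScreenContainerTabButton, with icon, collapsible title and "!" badge.
# Parent events play hover, reset, selection and touch animations driving icon material parameters, scale, title opacity and padding.
# BP_Selection's branch has no condition connected; BP_DelayedDeselect plays Unselected, then collapses the title.
# Tag category changes recolor the title; Tag_* animations are defined but unused in this graph.
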WidgetBlueprint /Game/UI/Sidebar/Tags/BP_SidebarCollectionScreenContainerTabButtonEditTags
animations [
 Selected {
  binding_count=3
  length_seconds=0.15
  object=&self:Selected
  sequence {bindings [
    Image_61 {
     id=#1
     display_name=LazyImage_Icon
     possessed_class=Image
     tracks [
      {
       type=MovieSceneWidgetMaterialTrack
       sections [{type=MovieSceneParameterSection, end_frame=3, start_frame=0, channels [IsHovered {type=MovieSceneScriptingFloatChannel, keys [[0,0],[3,1]]}]}]
      }
      {
       type=MovieScene2DTransformTrack
       sections [{
         type=MovieScene2DTransformSection
         end_frame=0
         start_frame=0
         channels [
          Translation.X {type=MovieSceneScriptingFloatChannel}
          Translation.Y {type=MovieSceneScriptingFloatChannel}
          Angle {type=MovieSceneScriptingFloatChannel}
          Scale.X {type=MovieSceneScriptingFloatChannel, const [1,1]}
          Scale.Y {type=MovieSceneScriptingFloatChannel, const [1,1]}
          Shear.X {type=MovieSceneScriptingFloatChannel}
          Shear.Y {type=MovieSceneScriptingFloatChannel}
         ]
        }]
      }
      {
       type=MovieSceneFloatTrack
       display_name=Render Opacity
       sections [{type=MovieSceneFloatSection, end_frame=3, start_frame=3, channels [None {type=MovieSceneScriptingFloatChannel, const [1,1]}]}]
      }
     ]
    }
    HorizontalBoxSlot_1 {
     id=#2
     parent_id=#3
     possessed_class=HorizontalBoxSlot
     tracks [{
       type=MovieSceneMarginTrack
       display_name=Padding
       sections [{
         type=MovieSceneMarginSection
         end_frame=3
         start_frame=0
         channels [
          Left {type=MovieSceneScriptingFloatChannel, keys [[0,-64],[3,10]]}
          Top {type=MovieSceneScriptingFloatChannel}
          Right {type=MovieSceneScriptingFloatChannel, keys [[2,0],[3,10]]}
          Bottom {type=MovieSceneScriptingFloatChannel}
         ]
        }]
      }]
    }
    Text_TabTitle {
     id=#3
     possessed_class=CommonRichTextBlock
     tracks [
      {
       type=MovieSceneFloatTrack
       display_name=Render Opacity
       sections [{type=MovieSceneFloatSection, end_frame=3, start_frame=0, channels [None {type=MovieSceneScriptingFloatChannel, keys [[2,0],[3,1]]}]}]
      }
      {
       type=MovieScene2DTransformTrack
       sections [{
         type=MovieScene2DTransformSection
         end_frame=3
         start_frame=0
         channels [
          Translation.X {type=MovieSceneScriptingFloatChannel}
          Translation.Y {type=MovieSceneScriptingFloatChannel}
          Angle {type=MovieSceneScriptingFloatChannel}
          Scale.X {type=MovieSceneScriptingFloatChannel, keys [[0,0],[2,1]]}
          Scale.Y {type=MovieSceneScriptingFloatChannel, const [1,2]}
          Shear.X {type=MovieSceneScriptingFloatChannel}
          Shear.Y {type=MovieSceneScriptingFloatChannel}
         ]
        }]
      }
     ]
    }
   ]}
 }
 Unselected {
  binding_count=3
  length_seconds=0.15
  object=&self:Unselected
  sequence {bindings [
    Text_TabTitle {
     id=#4
     possessed_class=CommonRichTextBlock
     tracks [
      {
       type=MovieSceneFloatTrack
       display_name=Render Opacity
       sections [{type=MovieSceneFloatSection, end_frame=3, start_frame=0, channels [None {type=MovieSceneScriptingFloatChannel, keys [[0,1],[1,0]]}]}]
      }
      {
       type=MovieScene2DTransformTrack
       sections [{
         type=MovieScene2DTransformSection
         end_frame=3
         start_frame=0
         channels [
          Translation.X {type=MovieSceneScriptingFloatChannel}
          Translation.Y {type=MovieSceneScriptingFloatChannel}
          Angle {type=MovieSceneScriptingFloatChannel}
          Scale.X {type=MovieSceneScriptingFloatChannel, keys [[0,1],[1,1],[3,0]]}
          Scale.Y {type=MovieSceneScriptingFloatChannel, const [1,3]}
          Shear.X {type=MovieSceneScriptingFloatChannel}
          Shear.Y {type=MovieSceneScriptingFloatChannel}
         ]
        }]
      }
     ]
    }
    Image_61 {
     id=#1
     display_name=LazyImage_Icon
     possessed_class=Image
     tracks [
      {
       type=MovieSceneWidgetMaterialTrack
       sections [{type=MovieSceneParameterSection, end_frame=3, start_frame=0, channels [IsHovered {type=MovieSceneScriptingFloatChannel, keys [[0,1],[3,0]]}]}]
      }
      {
       type=MovieScene2DTransformTrack
       sections [{
         type=MovieScene2DTransformSection
         end_frame=0
         start_frame=0
         channels [
          Translation.X {type=MovieSceneScriptingFloatChannel}
          Translation.Y {type=MovieSceneScriptingFloatChannel}
          Angle {type=MovieSceneScriptingFloatChannel}
          Scale.X {type=MovieSceneScriptingFloatChannel, const [1,1]}
          Scale.Y {type=MovieSceneScriptingFloatChannel, const [1,1]}
          Shear.X {type=MovieSceneScriptingFloatChannel}
          Shear.Y {type=MovieSceneScriptingFloatChannel}
         ]
        }]
      }
      {
       type=MovieSceneFloatTrack
       display_name=Render Opacity
       sections [{type=MovieSceneFloatSection, end_frame=3, start_frame=3, channels [None {type=MovieSceneScriptingFloatChannel, const [1,1]}]}]
      }
     ]
    }
    HorizontalBoxSlot_1 {
     id=#5
     parent_id=#4
     possessed_class=HorizontalBoxSlot
     tracks [{
       type=MovieSceneMarginTrack
       display_name=Padding
       sections [{
         type=MovieSceneMarginSection
         end_frame=3
         start_frame=0
         channels [
          Left {type=MovieSceneScriptingFloatChannel, keys [[0,10],[3,-64]]}
          Top {type=MovieSceneScriptingFloatChannel}
          Right {type=MovieSceneScriptingFloatChannel, keys [[0,10],[1,0]]}
          Bottom {type=MovieSceneScriptingFloatChannel}
         ]
        }]
      }]
    }
   ]}
 }
 Hovered {
  binding_count=1
  length_seconds=0.15
  object=&self:Hovered
  sequence {bindings [Image_61 {
     id=#1
     display_name=LazyImage_Icon
     possessed_class=Image
     tracks [
      {
       type=MovieScene2DTransformTrack
       sections [{
         type=MovieScene2DTransformSection
         channels [
          Translation.X {type=MovieSceneScriptingFloatChannel}
          Translation.Y {type=MovieSceneScriptingFloatChannel}
          Angle {type=MovieSceneScriptingFloatChannel}
          Scale.X {type=MovieSceneScriptingFloatChannel, keys [[0,1],[3,1.1]]}
          Scale.Y {type=MovieSceneScriptingFloatChannel, keys [[0,1],[3,1.1]]}
          Shear.X {type=MovieSceneScriptingFloatChannel}
          Shear.Y {type=MovieSceneScriptingFloatChannel}
         ]
        }]
      }
      {
       type=MovieSceneWidgetMaterialTrack
       sections [{type=MovieSceneParameterSection, end_frame=3, start_frame=0, channels [IsHovered {type=MovieSceneScriptingFloatChannel, const [0,1]}]}]
      }
     ]
    }]}
 }
 Unhovered {
  binding_count=1
  length_seconds=0.15
  object=&self:Unhovered
  sequence {bindings [Image_61 {
     id=#1
     display_name=LazyImage_Icon
     possessed_class=Image
     tracks [
      {
       type=MovieScene2DTransformTrack
       sections [{
         type=MovieScene2DTransformSection
         channels [
          Translation.X {type=MovieSceneScriptingFloatChannel}
          Translation.Y {type=MovieSceneScriptingFloatChannel}
          Angle {type=MovieSceneScriptingFloatChannel}
          Scale.X {type=MovieSceneScriptingFloatChannel, keys [[0,1.1],[3,1]]}
          Scale.Y {type=MovieSceneScriptingFloatChannel, keys [[0,1.1],[3,1]]}
          Shear.X {type=MovieSceneScriptingFloatChannel}
          Shear.Y {type=MovieSceneScriptingFloatChannel}
         ]
        }]
      }
      {
       type=MovieSceneWidgetMaterialTrack
       sections [{type=MovieSceneParameterSection, end_frame=3, start_frame=0, channels [IsHovered {type=MovieSceneScriptingFloatChannel, const [0,1]}]}]
      }
     ]
    }]}
 }
 Reset {
  binding_count=1
  length_seconds=0.1
  object=&self:Reset
  sequence {bindings [Image_61 {
     id=#1
     display_name=LazyImage_Icon
     possessed_class=Image
     tracks [
      {
       type=MovieScene2DTransformTrack
       sections [{
         type=MovieScene2DTransformSection
         channels [
          Translation.X {type=MovieSceneScriptingFloatChannel}
          Translation.Y {type=MovieSceneScriptingFloatChannel}
          Angle {type=MovieSceneScriptingFloatChannel}
          Scale.X {type=MovieSceneScriptingFloatChannel, const [1,1]}
          Scale.Y {type=MovieSceneScriptingFloatChannel, const [1,1]}
          Shear.X {type=MovieSceneScriptingFloatChannel}
          Shear.Y {type=MovieSceneScriptingFloatChannel}
         ]
        }]
      }
      {
       type=MovieSceneWidgetMaterialTrack
       sections [{
         type=MovieSceneParameterSection
         end_frame=1
         start_frame=-1
         channels [IsHovered {type=MovieSceneScriptingFloatChannel, keys [[-1,0],[0,-1]]}]
        }]
      }
     ]
    }]}
 }
 Tag_Reset {
  binding_count=1
  length_seconds=0.05
  object=&self:Tag_Reset
  sequence {bindings [Image_61 {
     id=#1
     display_name=LazyImage_Icon
     possessed_class=Image
     tracks [
      {
       type=MovieScene2DTransformTrack
       sections [{
         type=MovieScene2DTransformSection
         channels [
          Translation.X {type=MovieSceneScriptingFloatChannel}
          Translation.Y {type=MovieSceneScriptingFloatChannel}
          Angle {type=MovieSceneScriptingFloatChannel}
          Scale.X {type=MovieSceneScriptingFloatChannel, const [1,1]}
          Scale.Y {type=MovieSceneScriptingFloatChannel, const [1,1]}
          Shear.X {type=MovieSceneScriptingFloatChannel}
          Shear.Y {type=MovieSceneScriptingFloatChannel}
         ]
        }]
      }
      {
       type=MovieSceneWidgetMaterialTrack
       sections [{type=MovieSceneParameterSection, end_frame=1, start_frame=0, channels [IsHovered {type=MovieSceneScriptingFloatChannel, const [0,1]}]}]
      }
     ]
    }]}
 }
 Tag_Hovered {
  binding_count=1
  length_seconds=0.1
  object=&self:Tag_Hovered
  sequence {bindings [Image_61 {
     id=#1
     display_name=LazyImage_Icon
     possessed_class=Image
     tracks [{
       type=MovieSceneWidgetMaterialTrack
       sections [{type=MovieSceneParameterSection, end_frame=2, start_frame=0, channels [IsHovered {type=MovieSceneScriptingFloatChannel, keys [[0,0],[2,1]]}]}]
      }]
    }]}
 }
 Tag_Unhovered {
  binding_count=1
  length_seconds=0.1
  object=&self:Tag_Unhovered
  sequence {bindings [Image_61 {
     id=#1
     display_name=LazyImage_Icon
     possessed_class=Image
     tracks [{
       type=MovieSceneWidgetMaterialTrack
       sections [{type=MovieSceneParameterSection, end_frame=2, start_frame=0, channels [IsHovered {type=MovieSceneScriptingFloatChannel, keys [[0,1],[2,0]]}]}]
      }]
    }]}
 }
 Tag_Enabled {
  binding_count=1
  length_seconds=0.1
  object=&self:Tag_Enabled
  sequence {bindings [Image_61 {
     id=#6
     display_name=LazyImage_Icon
     possessed_class=Image
     tracks [{
       type=MovieSceneWidgetMaterialTrack
       sections [{type=MovieSceneParameterSection, end_frame=2, start_frame=0, channels [IsDisabled {type=MovieSceneScriptingFloatChannel, const [0,1]}]}]
      }]
    }]}
 }
 Tag_DisabledHovered {
  binding_count=1
  length_seconds=0.1
  object=&self:Tag_DisabledHovered
  sequence {bindings [Image_61 {
     id=#6
     display_name=LazyImage_Icon
     possessed_class=Image
     tracks [{
       type=MovieSceneWidgetMaterialTrack
       sections [{
         type=MovieSceneParameterSection
         end_frame=2
         start_frame=0
         channels [
          IsDisabled {type=MovieSceneScriptingFloatChannel, const [1,1]}
          DisabledHover {type=MovieSceneScriptingFloatChannel, const [1,1]}
         ]
        }]
      }]
    }]}
 }
 Tag_DisabledUnhovered {
  binding_count=1
  length_seconds=0.1
  object=&self:Tag_DisabledUnhovered
  sequence {bindings [Image_61 {
     id=#6
     display_name=LazyImage_Icon
     possessed_class=Image
     tracks [{
       type=MovieSceneWidgetMaterialTrack
       sections [{
         type=MovieSceneParameterSection
         end_frame=2
         start_frame=0
         channels [
          IsDisabled {type=MovieSceneScriptingFloatChannel, const [1,1]}
          DisabledHover {type=MovieSceneScriptingFloatChannel, const [0,1]}
         ]
        }]
      }]
    }]}
 }
 Tag_DisabledHover {
  binding_count=1
  length_seconds=0.1
  object=&self:Tag_DisabledHover
  sequence {bindings [Image_61 {
     id=#6
     display_name=LazyImage_Icon
     possessed_class=Image
     tracks [{
       type=MovieSceneWidgetMaterialTrack
       sections [{
         type=MovieSceneParameterSection
         end_frame=2
         start_frame=0
         channels [
          IsDisabled {type=MovieSceneScriptingFloatChannel, const [1,1]}
          DisabledHover {type=MovieSceneScriptingFloatChannel, keys [[0,0],[2,1]]}
         ]
        }]
      }]
    }]}
 }
 Tag_DisabledUnhover {
  binding_count=1
  length_seconds=0.1
  object=&self:Tag_DisabledUnhover
  sequence {bindings [Image_61 {
     id=#6
     display_name=LazyImage_Icon
     possessed_class=Image
     tracks [{
       type=MovieSceneWidgetMaterialTrack
       sections [{
         type=MovieSceneParameterSection
         end_frame=2
         start_frame=0
         channels [
          IsDisabled {type=MovieSceneScriptingFloatChannel, const [1,1]}
          DisabledHover {type=MovieSceneScriptingFloatChannel, keys [[0,1],[2,0]]}
         ]
        }]
      }]
    }]}
 }
 Unselected_Touch {
  binding_count=3
  length_seconds=0.01
  object=&self:Unselected_Touch
  sequence {bindings [
    Text_TabTitle {
     id=#4
     possessed_class=CommonRichTextBlock
     tracks [
      {
       type=MovieSceneFloatTrack
       display_name=Render Opacity
       sections [{type=MovieSceneFloatSection, end_frame=0, start_frame=0, channels [None {type=MovieSceneScriptingFloatChannel, const [0.3,1]}]}]
      }
      {
       type=MovieScene2DTransformTrack
       sections [{
         type=MovieScene2DTransformSection
         end_frame=0
         start_frame=0
         channels [
          Translation.X {type=MovieSceneScriptingFloatChannel}
          Translation.Y {type=MovieSceneScriptingFloatChannel}
          Angle {type=MovieSceneScriptingFloatChannel}
          Scale.X {type=MovieSceneScriptingFloatChannel, const [1,1]}
          Scale.Y {type=MovieSceneScriptingFloatChannel, const [1,1]}
          Shear.X {type=MovieSceneScriptingFloatChannel}
          Shear.Y {type=MovieSceneScriptingFloatChannel}
         ]
        }]
      }
     ]
    }
    Image_61 {
     id=#1
     display_name=LazyImage_Icon
     possessed_class=Image
     tracks [
      {
       type=MovieSceneWidgetMaterialTrack
       sections [{type=MovieSceneParameterSection, end_frame=0, start_frame=0, channels [IsHovered {type=MovieSceneScriptingFloatChannel, const [1,1]}]}]
      }
      {
       type=MovieScene2DTransformTrack
       sections [{
         type=MovieScene2DTransformSection
         end_frame=0
         start_frame=0
         channels [
          Translation.X {type=MovieSceneScriptingFloatChannel}
          Translation.Y {type=MovieSceneScriptingFloatChannel}
          Angle {type=MovieSceneScriptingFloatChannel}
          Scale.X {type=MovieSceneScriptingFloatChannel, const [1,1]}
          Scale.Y {type=MovieSceneScriptingFloatChannel, const [1,1]}
          Shear.X {type=MovieSceneScriptingFloatChannel}
          Shear.Y {type=MovieSceneScriptingFloatChannel}
         ]
        }]
      }
      {
       type=MovieSceneFloatTrack
       display_name=Render Opacity
       sections [{type=MovieSceneFloatSection, end_frame=0, start_frame=0, channels [None {type=MovieSceneScriptingFloatChannel, const [0.3,1]}]}]
      }
     ]
    }
    HorizontalBoxSlot_1 {
     id=#5
     parent_id=#4
     possessed_class=HorizontalBoxSlot
     tracks [{
       type=MovieSceneMarginTrack
       display_name=Padding
       sections [{
         type=MovieSceneMarginSection
         end_frame=0
         start_frame=0
         channels [
          Left {type=MovieSceneScriptingFloatChannel, const [10,1]}
          Top {type=MovieSceneScriptingFloatChannel}
          Right {type=MovieSceneScriptingFloatChannel, const [10,1]}
          Bottom {type=MovieSceneScriptingFloatChannel}
         ]
        }]
      }]
    }
   ]}
 }
 Selected_Touch {
  binding_count=3
  length_seconds=0.01
  object=&self:Selected_Touch
  sequence {bindings [
    Image_61 {
     id=#1
     display_name=LazyImage_Icon
     possessed_class=Image
     tracks [
      {
       type=MovieSceneWidgetMaterialTrack
       sections [{type=MovieSceneParameterSection, end_frame=0, start_frame=0, channels [IsHovered {type=MovieSceneScriptingFloatChannel, const [1,1]}]}]
      }
      {
       type=MovieScene2DTransformTrack
       sections [{
         type=MovieScene2DTransformSection
         end_frame=0
         start_frame=0
         channels [
          Translation.X {type=MovieSceneScriptingFloatChannel}
          Translation.Y {type=MovieSceneScriptingFloatChannel}
          Angle {type=MovieSceneScriptingFloatChannel}
          Scale.X {type=MovieSceneScriptingFloatChannel, const [1,1]}
          Scale.Y {type=MovieSceneScriptingFloatChannel, const [1,1]}
          Shear.X {type=MovieSceneScriptingFloatChannel}
          Shear.Y {type=MovieSceneScriptingFloatChannel}
         ]
        }]
      }
      {
       type=MovieSceneFloatTrack
       display_name=Render Opacity
       sections [{type=MovieSceneFloatSection, end_frame=0, start_frame=0, channels [None {type=MovieSceneScriptingFloatChannel, const [1,1]}]}]
      }
     ]
    }
    HorizontalBoxSlot_1 {
     id=#2
     parent_id=#3
     possessed_class=HorizontalBoxSlot
     tracks [{
       type=MovieSceneMarginTrack
       display_name=Padding
       sections [{
         type=MovieSceneMarginSection
         end_frame=0
         start_frame=0
         channels [
          Left {type=MovieSceneScriptingFloatChannel, const [10,1]}
          Top {type=MovieSceneScriptingFloatChannel}
          Right {type=MovieSceneScriptingFloatChannel, const [10,1]}
          Bottom {type=MovieSceneScriptingFloatChannel}
         ]
        }]
      }]
    }
    Text_TabTitle {
     id=#3
     possessed_class=CommonRichTextBlock
     tracks [
      {
       type=MovieSceneFloatTrack
       display_name=Render Opacity
       sections [{type=MovieSceneFloatSection, end_frame=0, start_frame=0, channels [None {type=MovieSceneScriptingFloatChannel, const [1,1]}]}]
      }
      {
       type=MovieScene2DTransformTrack
       sections [{
         type=MovieScene2DTransformSection
         end_frame=0
         start_frame=0
         channels [
          Translation.X {type=MovieSceneScriptingFloatChannel}
          Translation.Y {type=MovieSceneScriptingFloatChannel}
          Angle {type=MovieSceneScriptingFloatChannel}
          Scale.X {type=MovieSceneScriptingFloatChannel, const [1,1]}
          Scale.Y {type=MovieSceneScriptingFloatChannel, const [1,1]}
          Shear.X {type=MovieSceneScriptingFloatChannel}
          Shear.Y {type=MovieSceneScriptingFloatChannel}
         ]
        }]
      }
     ]
    }
   ]}
 }
]
graphs [EventGraph {nodes [
   n1 = Event BP_Hover!() -> (OutputDelegate, then, bIsHovered:bool) {is=title K2Node_Event}
   > n2 = Branch(Condition=@n1.bIsHovered)
   > PlayAnimationForward(InAnimation=Get Hovered() -> (Hovered:Widget Animation Object Reference))
   PlayAnimationForward(execute=@n2.else, InAnimation=Get Unhovered() -> (Unhovered:Widget Animation Object Reference))
   Event BP_DelayedDeselect() {is=title K2Node_Event}
   > n3 = "Play Animation with Finished event (legacy)"!(
    execute:Exec
    Widget:User Widget Object Reference=Self-Reference()
    InAnimation:Widget Animation Object Reference=Get Unselected() -> (Unselected:Widget Animation Object Reference)
    StartAtTime:float=0
    NumLoopsToPlay:int=1
    PlayMode:EUMGSequencePlayMode Enum
    PlaybackSpeed:float=1
   ) -> (then:Exec, Result:UMGSequence Player Object Reference, Finished:Exec) {is=title K2Node_PlayAnimation}
   SetVisibility(execute=@n3.Finished, self=Get Text_TabTitle() -> (Text_TabTitle:Rich Text Block Object Reference), InVisibility=Collapsed)
   n4 = Event BP_OnTagCategoryChange!() -> (OutputDelegate, then, InTagCategory:ESocialTagCategory Enum=GameModes) {is=title K2Node_Event}
   > n5 = Get Sidebar Social Tag Category Structure!(execute, TagCategory:String=Enum to String!(Enumerator:ESocialTagCategory Enum=@n4.InTagCategory) -> (ReturnValue:String)) -> (then, SidebarSocialTagCategoryStructure:Sidebar Social Tag Category Structure Structure) {is=title K2Node_CallFunction}
   > SetDefaultColorAndOpacity!(
    execute
    self:Rich Text Block Object Reference=Get Text_TabTitle()
    InColorAndOpacity:Slate Color Structure=Break Sidebar Social Tag Category Structure!(SidebarSocialTagCategoryStructure:"Sidebar Social Tag Category Structure Structure (by ref)"=@n5) -> (
     TagSkewBoxInnerColor_48_#7:Linear Color Structure
     TagSkewBoxInnerColorHovered_42_#8:Linear Color Structure
     TagSkewBoxInnerColorSelected_45_#9:Linear Color Structure
     TagSkewBoxOuterColor_41_#10:Linear Color Structure
     TagSkewBoxOuterColorSelected_47_#11:Linear Color Structure
     TagCategoryBaseColor_19_#12:Slate Color Structure
    ) {is=title K2Node_BreakStruct}.TagCategoryBaseColor_19_#12
   ) -> (then)
   Event BP_Reset() {is=title K2Node_Event}
   > PlayAnimationForward(InAnimation=Get Reset() -> (Reset:Widget Animation Object Reference))
   n6 = Event BP_SelectionTouch!() -> (OutputDelegate, then, bisSelected:bool) {is=title K2Node_Event}
   > n7 = Branch(Condition=@n6.bisSelected)
   > PlayAnimationForward(InAnimation=Get Selected_Touch() -> (Selected_Touch:Widget Animation Object Reference))
   PlayAnimationForward(execute=@n7.else, InAnimation=Get Unselected_Touch() -> (Unselected_Touch:Widget Animation Object Reference))
   Event BP_Selection!() -> (OutputDelegate, then, bIsSelected:bool) {is=title K2Node_Event}
   > n8 = Branch()
   > PlayAnimationForward(InAnimation=Get Selected() -> (Selected:Widget Animation Object Reference))
   PlayAnimationForward(execute=@n8.else, InAnimation=Get Unselected())
  ]}]
variables []
widgets {
 info {
  inheritedWidgetCount=6
  parentClass=/Script/FortniteUI.FortSidebarCollectionScreenContainerTabButton
  rootWidgetClass=&ScaleBox
  widgetCount=9
 }
 widgets [
  ScaleBox_MobileMultiplier {
   widgetClassPath=&ScaleBox
   bInherited=true
   bIsVariable=true
   widget=ScaleBox_MobileMultiplier
   properties {stretch=UserSpecified}
  }
  Overlay_0 {widgetClassPath=&Overlay, parent=ScaleBox_MobileMultiplier, widget=Overlay_0, slot {class=ScaleBoxSlot}}
  HorizontalBox_0 {
   widgetClassPath=&HorizontalBox
   parent=Overlay_0
   widget=HorizontalBox_0
   slot {class=OverlaySlot, properties {horizontalAlignment=HAlign_Fill, verticalAlignment=VAlign_Fill}}
  }
  LazyImage_Icon {
   widgetClassPath=&FortLazyImage
   bInherited=true
   bIsVariable=true
   parent=HorizontalBox_0
   widget=LazyImage_Icon
   properties {brush {
     bIsDynamicallyLoaded=false
     drawAs=Image
     imageType=NoImage
     mirroring=NoMirror
     resourceName=None
     resourceObject=/Game/UI/Sidebar/Tags/Materials/SocialTagCategory_GameMode
     tiling=NoTile
     imageSize=<80 80>
     margin=<ltrb 0 0 0 0>
     outlineSettings {
      bUseBrushTransparency=false
      roundingType=HalfHeightRadius
      width=0
      color {colorUseRule=UseColor_Specified, specifiedColor=<rgba 0 0 0 0>}
      cornerRadii=<0 0 0 0>
     }
     tintColor {colorUseRule=UseColor_Specified, specifiedColor=<rgba 1 1 1 1>}
     uVRegion {bIsValid=false, max=<0 0>, min=<0 0>}
    }}
   slot {class=HorizontalBoxSlot, properties {horizontalAlignment=HAlign_Center, verticalAlignment=VAlign_Center}}
  }
  Text_TabTitle {
   widgetClassPath=&CommonRichTextBlock
   bInherited=true
   bIsVariable=true
   parent=HorizontalBox_0
   widget=Text_TabTitle
   properties {
    bOverrideDefaultStyle=true
    mobileTextBlockScale=1.15
    text=Tab Title
    textTransformPolicy=ToUpper
    visibility=Collapsed
    defaultTextStyleOverride {
     overflowPolicy=Clip
     transformPolicy=None
     colorAndOpacity {colorUseRule=UseColor_Specified, specifiedColor=<rgba 1 1 1 1>}
     font {
      bForceMonospaced=false
      bMaterialIsStencil=false
      fontMaterial=null
      fontObject=&Burbank
      letterSpacing=0
      monospacedWidth=1
      size=16
      skewAmount=0
      typefaceFontName=Burbank-PlayerPower
      outlineSettings {
       bApplyOutlineToDropShadows=false
       bMiteredCorners=false
       bSeparateFillAlpha=false
       outlineMaterial=null
       outlineSize=0
       outlineColor=<rgba 0 0 0 1>
      }
     }
     highlightColor {colorUseRule=UseColor_Specified, specifiedColor=<rgba 0 0 0 1>}
     highlightShape {
      bIsDynamicallyLoaded=false
      drawAs=Image
      imageType=NoImage
      mirroring=NoMirror
      resourceName=None
      resourceObject=null
      tiling=NoTile
      imageSize=<32 32>
      margin=<ltrb 0 0 0 0>
      outlineSettings {
       bUseBrushTransparency=false
       roundingType=HalfHeightRadius
       width=0
       color {colorUseRule=UseColor_Specified, specifiedColor=<rgba 0 0 0 0>}
       cornerRadii=<0 0 0 0>
      }
      tintColor {colorUseRule=UseColor_Specified, specifiedColor=<rgba 1 1 1 1>}
      uVRegion {bIsValid=false, max=<0 0>, min=<0 0>}
     }
     selectedBackgroundColor {colorUseRule=UseColor_Foreground, specifiedColor=<rgba 1 0 1 1>}
     shadowColorAndOpacity=<rgba 0 0 0 1>
     shadowOffset=<0 0>
     strikeBrush {
      bIsDynamicallyLoaded=false
      drawAs=Image
      imageType=NoImage
      mirroring=NoMirror
      resourceName=None
      resourceObject=null
      tiling=NoTile
      imageSize=<32 32>
      margin=<ltrb 0 0 0 0>
      outlineSettings {
       bUseBrushTransparency=false
       roundingType=HalfHeightRadius
       width=0
       color {colorUseRule=UseColor_Specified, specifiedColor=<rgba 0 0 0 0>}
       cornerRadii=<0 0 0 0>
      }
      tintColor {colorUseRule=UseColor_Specified, specifiedColor=<rgba 1 1 1 1>}
      uVRegion {bIsValid=false, max=<0 0>, min=<0 0>}
     }
     underlineBrush {
      bIsDynamicallyLoaded=false
      drawAs=Image
      imageType=NoImage
      mirroring=NoMirror
      resourceName=None
      resourceObject=null
      tiling=NoTile
      imageSize=<32 32>
      margin=<ltrb 0 0 0 0>
      outlineSettings {
       bUseBrushTransparency=false
       roundingType=HalfHeightRadius
       width=0
       color {colorUseRule=UseColor_Specified, specifiedColor=<rgba 0 0 0 0>}
       cornerRadii=<0 0 0 0>
      }
      tintColor {colorUseRule=UseColor_Specified, specifiedColor=<rgba 1 1 1 1>}
      uVRegion {bIsValid=false, max=<0 0>, min=<0 0>}
     }
    }
    renderTransform {angle=0, scale=<1 1>, shear=<-8 0>, translation=<0 0>}
   }
   slot {class=HorizontalBoxSlot, properties {verticalAlignment=VAlign_Center, padding=<ltrb -9.1852 0 0 0>, size {sizeRule=Fill, value=1}}}
  }
  Mobile_Spacer {
   widgetClassPath=&Spacer
   bInherited=true
   bIsVariable=true
   parent=HorizontalBox_0
   widget=Mobile_Spacer
   properties {visibility=Collapsed, size=<25 0>}
   slot {class=HorizontalBoxSlot}
  }
  NewBang {
   widgetClassPath=&CommonBorder
   bInherited=true
   bIsVariable=true
   parent=Overlay_0
   widget=NewBang
   properties {
    horizontalAlignment=HAlign_Center
    visibility=Collapsed
    background {
     bIsDynamicallyLoaded=false
     drawAs=Box
     imageType=NoImage
     mirroring=NoMirror
     resourceName=None
     resourceObject=/Game/Athena/UI/ReplayBrowser/M_Ui_Skewed_Vfor9slices_Bang
     tiling=NoTile
     imageSize=<32 32>
     margin=<ltrb 0.5 0 0.5 0>
     outlineSettings {
      bUseBrushTransparency=false
      roundingType=HalfHeightRadius
      width=0
      color {colorUseRule=UseColor_Specified, specifiedColor=<rgba 0 0 0 0>}
      cornerRadii=<0 0 0 0>
     }
     tintColor {colorUseRule=UseColor_Specified, specifiedColor=<rgba 0.9301 1 0.0097 1>}
     uVRegion {bIsValid=false, max=<0 0>, min=<0 0>}
    }
    padding=<ltrb 7 6 7 6>
   }
   slot {class=OverlaySlot, properties {padding=<ltrb 64 -12 0 0>}}
  }
  CommonTextBlock_209 {
   widgetClassPath=&CommonTextBlock
   parent=NewBang
   widget=CommonTextBlock_209
   properties {
    lineHeightPercentage=0.7
    text="!"
    colorAndOpacity {colorUseRule=UseColor_Specified, specifiedColor=<rgba 0 0 0 1>}
    font {
     bForceMonospaced=false
     bMaterialIsStencil=false
     fontMaterial=null
     fontObject=&Burbank
     letterSpacing=0
     monospacedWidth=1
     size=16
     skewAmount=0
     typefaceFontName=BurbankRegular-Black
     outlineSettings {
      bApplyOutlineToDropShadows=false
      bMiteredCorners=false
      bSeparateFillAlpha=false
      outlineMaterial=null
      outlineSize=0
      outlineColor=<rgba 0.9301 1 0.0097 1>
     }
    }
    margin=<ltrb 0 -5 0 0>
    renderTransform {angle=0, scale=<1 1>, shear=<-12 0>, translation=<0 0>}
   }
   slot {class=BorderSlot, properties {horizontalAlignment=HAlign_Center, padding=<ltrb 7 6 7 6>}}
  }
  InputActionWidget {widgetClassPath=&CommonActionWidget, bInherited=true, widget=null}
 ]
}
