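# Sidebar social tag widget that shows an "Adding Tag" placeholder: a skewed, material-backed label with a direction arrow above it.
# On Construct, EventSetTag looks up the "AddTag" category structure and copies its inner and selected outer colors into the border's dynamic material parameters.
# The same function starts the AddingTag animation, which loops indefinitely to bob the arrow vertically.
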WidgetBlueprint /Game/UI/Sidebar/Tags/SidebarSocialAddingTagWidget
animations [AddingTag {
  binding_count=1
  length_seconds=1
  object=&self:AddingTag
  sequence {bindings [Image_AddingTagArrow {
     id=#1
     possessed_class=Image
     tracks [{
       type=MovieScene2DTransformTrack
       sections [{
         type=MovieScene2DTransformSection
         end_frame=20
         start_frame=0
         channels [
          Translation.X {type=MovieSceneScriptingFloatChannel, const [0,4]}
          Translation.Y {type=MovieSceneScriptingFloatChannel, keys [[0,0],[5,2],[10,0],[20,0]]}
          Angle {type=MovieSceneScriptingFloatChannel}
          Scale.X {type=MovieSceneScriptingFloatChannel}
          Scale.Y {type=MovieSceneScriptingFloatChannel}
          Shear.X {type=MovieSceneScriptingFloatChannel}
          Shear.Y {type=MovieSceneScriptingFloatChannel}
         ]
        }]
      }]
    }]}
 }]
graphs [
 EventGraph {nodes [
   Event Construct()
   > EventSetTag()
  ]}
 EventSetTag {nodes [
   EventSetTag() {is=title K2Node_FunctionEntry}
   > n1 = Sequence()
   > n2 = Get Sidebar Social Tag Category Structure!(execute, TagCategory:String=Get TagCategory_AddTag() -> (TagCategory_AddTag:String)) -> (then, SidebarSocialTagCategoryStructure:Sidebar Social Tag Category Structure Structure) {is=title K2Node_CallFunction}
   > n3 = GetDynamicMaterial(self:Border Object Reference=Get Border_AddingTagBorder() -> (Border_AddingTagBorder:Border Object Reference))
   n4 = Break Sidebar Social Tag Category Structure!(SidebarSocialTagCategoryStructure:"Sidebar Social Tag Category Structure Structure (by ref)"=@n2) -> (
    TagSkewBoxInnerColor_48_#2:Linear Color Structure
    TagSkewBoxOuterColorSelected_47_#3:Linear Color Structure
   ) {is=title K2Node_BreakStruct}
   > SetVectorParameterValue(self=@n3, ParameterName=Get Name_InnerColor() -> (Name_InnerColor:Name=None), Value=@n4)
   > SetVectorParameterValue(
    self=@n3
    ParameterName=Get OuterColor() -> (OuterColor:Name=None)
    Value=@n4.TagSkewBoxOuterColorSelected_47_#3
   )
   PlayAnimation(execute=@n1.then_1, InAnimation=Get AddingTag() -> (AddingTag:Widget Animation Object Reference), NumLoopsToPlay=0)
  ]}
]
variables [
 Name_InnerColor {type=string, default=InnerColor}
 OuterColor {type=string, default=OuterColor}
 TagCategory_AddTag {type=string, default=AddTag}
]
widgets {
 info {inheritedWidgetCount=0, parentClass=&CommonUserWidget, rootWidgetClass=&Overlay, widgetCount=5}
 widgets [
  Overlay_AddingTag {widgetClassPath=&Overlay, bIsVariable=true, widget=Overlay_AddingTag}
  Image_AddingTagArrow {
   widgetClassPath=&Image
   bIsVariable=true
   parent=Overlay_AddingTag
   widget=Image_AddingTagArrow
   properties {
    brush {
     bIsDynamicallyLoaded=false
     drawAs=Image
     imageType=NoImage
     mirroring=NoMirror
     resourceName=None
     resourceObject=/Game/Widgets/SActorCanvas/DirectionArrow
     tiling=NoTile
     imageSize=<16 10>
     margin=<ltrb 0 0 0 0>
     outlineSettings {
      bUseBrushTransparency=false
      roundingType=HalfHeightRadius
      width=0
      color {colorUseRule=UseColor_Specified, specifiedColor=<rgba 0 0 0 0>}
      cornerRadii=<0 0 0 0>
     }
     tintColor {colorUseRule=UseColor_Specified, specifiedColor=<rgba 1 1 1 1>}
     uVRegion {bIsValid=false, max=<0 0>, min=<0 0>}
    }
    renderTransform {angle=180, scale=<1 1>, shear=<0 0>, translation=<0 0>}
   }
   slot {class=OverlaySlot, properties {horizontalAlignment=HAlign_Center, padding=<ltrb 0 -12 0 0>}}
  }
  Border_AddingTagBorder {
   widgetClassPath=&Border
   bIsVariable=true
   parent=Overlay_AddingTag
   widget=Border_AddingTagBorder
   properties {
    background {
     bIsDynamicallyLoaded=false
     drawAs=Image
     imageType=NoImage
     mirroring=NoMirror
     resourceName=None
     resourceObject=/Game/UI/Sidebar/Tags/Materials/BoxSkew_SocialTag
     tiling=NoTile
     imageSize=<32 32>
     margin=<ltrb 0 0 0 0>
     outlineSettings {
      bUseBrushTransparency=false
      roundingType=HalfHeightRadius
      width=0
      color {colorUseRule=UseColor_Specified, specifiedColor=<rgba 0 0 0 0>}
      cornerRadii=<0 0 0 0>
     }
     tintColor {colorUseRule=UseColor_Specified, specifiedColor=<rgba 1 1 1 1>}
     uVRegion {bIsValid=false, max=<0 0>, min=<0 0>}
    }
    padding=<ltrb 8 6 12 6>
   }
   slot {class=OverlaySlot, properties {horizontalAlignment=HAlign_Fill}}
  }
  HBox_AddingTag {
   widgetClassPath=&HorizontalBox
   parent=Border_AddingTagBorder
   widget=HBox_AddingTag
   properties {clipping=ClipToBounds}
   slot {class=BorderSlot, properties {padding=<ltrb 8 6 12 6>}}
  }
  Text_AddingTagName {
   widgetClassPath=&CommonTextBlock
   bIsVariable=true
   parent=HBox_AddingTag
   widget=Text_AddingTagName
   properties {
    mobileFontSizeMultiplier=1.25
    text=Adding Tag
    textOverflowPolicy=Ellipsis
    colorAndOpacity {colorUseRule=UseColor_Specified, specifiedColor=<rgba 1 1 1 0.8>}
    font {
     bForceMonospaced=false
     bMaterialIsStencil=false
     fontMaterial=null
     fontObject=&BurbankSmall
     letterSpacing=0
     monospacedWidth=1
     size=16
     skewAmount=0.15
     typefaceFontName=Black
     outlineSettings {
      bApplyOutlineToDropShadows=false
      bMiteredCorners=false
      bSeparateFillAlpha=false
      outlineMaterial=null
      outlineSize=0
      outlineColor=<rgba 0 0 0 1>
     }
    }
    shadowOffset=<0 0>
   }
   slot {class=HorizontalBoxSlot, properties {verticalAlignment=VAlign_Center, padding=<ltrb 10 0 10 0>, size {sizeRule=Fill, value=1}}}
  }
 ]
}
