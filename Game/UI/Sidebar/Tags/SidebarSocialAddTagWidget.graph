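# Sidebar social "+Add Tag" button (CommonButtonBase) that switches to the SidebarSocialAddingTagWidget when selected and back when deselected.
# On construct it loads AddTag and Disabled category colors and applies them to the skewed-box border's dynamic material.
# Hover and unhover set target colors and toggle text scrolling; the Hovered animation drives a lerp alpha alongside a looping timer for text and border color blending.
# A touch-only overlay extends the hit zone vertically.
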WidgetBlueprint /Game/UI/Sidebar/Tags/SidebarSocialAddTagWidget
animations [Hovered {
  binding_count=3
  length_seconds=0.1
  object=&self:Hovered
  sequence {bindings [SidebarSocialAddTagWidget_C_49 {
     id=#1
     possessed_class=SidebarSocialAddTagWidget_C
     tracks [{
       type=MovieSceneFloatTrack
       display_name=State Change Lerp Alpha
       sections [{type=MovieSceneFloatSection, end_frame=2, start_frame=0, channels [None {type=MovieSceneScriptingFloatChannel, keys [[0,0],[2,1]]}]}]
      }]
    }]}
 }]
graphs [
 EventGraph {nodes [
   "On Input Icon Updated (InputActionWidget)"() {is=title K2Node_ComponentBoundEvent}
   > n1 = GetIconDynamicMaterial!(execute, self:Common Action Widget Object Reference=Get InputActionWidget() -> (InputActionWidget:Common Action Widget Object Reference)) -> (then, ReturnValue:Material Instance Dynamic Object Reference)
   > SetScalarParameterValue(self=@n1, ParameterName=Use Dark, Value=1)
   > SetScalarParameterValue(self=@n1, ParameterName=Outline Opacity, Value=0)
   InterpolateTagColors() {is=title K2Node_CustomEvent}
   n2 = Get StateChangeLerpAlpha() -> (StateChangeLerpAlpha:double)
   > InterpolateTagContentColors!(
    execute
    TagName:Text Object Reference=Get Text_AddTagName() -> (Text_AddTagName:Common Text Object Reference)
    CheckMark:Image Object Reference
    CurrentTagContentColor:Linear Color Structure=Get CurrentTagContentColorReference() -> (CurrentTagContentColorReference:Linear Color Structure)
    TargetTagContentColor:Linear Color Structure=Get TargetTagContentColorReference() -> (TargetTagContentColorReference:Linear Color Structure)
    HoveredTagContentLerpAlpha:double=@n2
   ) -> (then)
   > n3 = Get!(execute:Exec) -> (then:Exec, else:Exec, TagBorderMaterialReference:Material Instance Dynamic Object Reference)
   > InterpolateTagBorderColors!(
    execute
    TagBorderMaterial:Material Instance Dynamic Object Reference=@n3
    NameInnerColor:Name=Get Name_InnerColor() -> (Name_InnerColor:Name=None)
    CurrentInnerColor:Linear Color Structure=Get CurrentInnerColorReference() -> (CurrentInnerColorReference:Linear Color Structure)
    TargetInnerColor:Linear Color Structure=Get TargetInnerColorReference() -> (TargetInnerColorReference:Linear Color Structure)
    NameOuterColor:Name=Get Name_OuterColor() -> (Name_OuterColor:Name=None)
    CurrentOuterColor:Linear Color Structure=Get CurrentOuterColorReference() -> (CurrentOuterColorReference:Linear Color Structure)
    TargetOuterColor:Linear Color Structure=Get TargetOuterColorReference() -> (TargetOuterColorReference:Linear Color Structure)
    HoveredBorderColorLerpAlpha:double=@n2
    ColorLerpTimeHandle:"Timer Handle Structure (by ref)"=Get StateChangeLerpTimerHandle() -> (StateChangeLerpTimerHandle:Timer Handle Structure)
    WorldContext:Object Reference=Self-Reference()
   ) -> (then)
   Event Construct()
   > EventSetTag()
   Event On Hovered()
   > FunctionOnHovered()
   Event On Unhovered()
   > FunctionOnUnhovered()
   Event On Selected()
   n4 = Get Switcher_Tag() -> (Switcher_Tag:Common Visibility Switcher Object Reference)
   > SetActiveWidget(self=@n4, Widget=Get SidebarSocialAddingTagWidget() -> (SidebarSocialAddingTagWidget:Sidebar Social Adding Tag Widget Object Reference))
   Event On Deselected()
   > SetActiveWidget(self=@n4, Widget=Get Border_AddTagBorder() -> (Border_AddTagBorder:Border Object Reference))
  ]}
 EventSetTag {nodes [
   EventSetTag() {is=title K2Node_FunctionEntry}
   > n1 = Sequence!(execute) -> (then_0, then_1, then_2:Exec, then_3:Exec)
   SetColorAndOpacity(
    execute=@n1.then_2
    self=Get Text_AddTagName()
    InColorAndOpacity=Make Slate Color(SpecifiedColor=Get TagContentColorWhite() -> (TagContentColorWhite:Linear Color Structure))
   )
   SetScrollingEnabled!(execute=@n1.then_0, self:Common Text Object Reference=Get Text_AddTagName(), bInIsScrollingEnabled:bool=false) -> (then)
   > n2 = GetDynamicMaterial(self:Border Object Reference=Get Border_AddTagBorder())
   > Is Valid(InputObject=@n2)
   > Set TagBorderMaterialReference(TagBorderMaterialReference:Material Instance Dynamic Object Reference=@n2) -> (Output_Get:Material Instance Dynamic Object Reference)
   n3 = Get Sidebar Social Tag Category Structure!(execute=@n1.then_1, TagCategory:String=Get TagCategory_AddTag() -> (TagCategory_AddTag:String)) -> (then, SidebarSocialTagCategoryStructure:Sidebar Social Tag Category Structure Structure) {is=title K2Node_CallFunction}
   > Set TagCategoryStructureReference(TagCategoryStructureReference:Sidebar Social Tag Category Structure Structure=@n3) -> (Output_Get:Sidebar Social Tag Category Structure Structure)
   > n4 = Get Sidebar Social Tag Category Structure(TagCategory=Get TagCategory_Disabled() -> (TagCategory_Disabled:String))
   > Set DisabledTagCategoryStructureReference(DisabledTagCategoryStructureReference:Sidebar Social Tag Category Structure Structure=@n4) -> (Output_Get:Sidebar Social Tag Category Structure Structure)
   n5 = GetDynamicMaterial(execute=@n1.then_3, self=Get Border_AddTagBorder())
   > Is Valid(InputObject=@n5)
   > n6 = Branch()
   n7 = Break Sidebar Social Tag Category Structure!(SidebarSocialTagCategoryStructure:"Sidebar Social Tag Category Structure Structure (by ref)"=Get TagCategoryStructureReference() -> (TagCategoryStructureReference:Sidebar Social Tag Category Structure Structure)) -> (TagSkewBoxInnerColor_48_#2:Linear Color Structure) {is=title K2Node_BreakStruct}
   n8 = Get Name_InnerColor() -> (Name_InnerColor=None)
   > SetVectorParameterValue(self=@n5, ParameterName=@n8, Value=@n7)
   n9 = Get Name_OuterColor() -> (Name_OuterColor=None)
   > SetVectorParameterValue(self=@n5, ParameterName=@n9, Value=@n7)
   n10 = Break Sidebar Social Tag Category Structure!(SidebarSocialTagCategoryStructure=Get DisabledTagCategoryStructureReference() -> (DisabledTagCategoryStructureReference:Sidebar Social Tag Category Structure Structure)) -> (TagSkewBoxInnerColor_48_#2, TagSkewBoxOuterColor_41_#3:Linear Color Structure)
   SetVectorParameterValue(execute=@n6.else, self=@n5, ParameterName=@n8, Value=@n10)
   > SetVectorParameterValue(self=@n5, ParameterName=@n9, Value=@n10.TagSkewBoxOuterColor_41_#3)
  ]}
 FunctionOnHovered {nodes [
   FunctionOnHovered() {is=title K2Node_FunctionEntry}
   > n1 = Sequence()
   > StopAllAnimations(self:User Widget Object Reference)
   > PlayAnimation(InAnimation=Get Hovered() -> (Hovered:Widget Animation Object Reference))
   Clear and Invalidate Timer by Handle(execute=@n1.then_1, Handle=Get StateChangeLerpTimerHandle())
   > GetCurrentTagBorderColors()
   > Set CurrentTagContentColorReference(CurrentTagContentColorReference:Linear Color Structure=Get TagContentColorWhite()) -> (Output_Get:Linear Color Structure)
   > Set TargetTagContentColorReference(TargetTagContentColorReference:Linear Color Structure=Get TagContentColorBlack() -> (TagContentColorBlack:Linear Color Structure)) -> (Output_Get:Linear Color Structure)
   n2 = Branch(execute=@n1.then_2)
   n3 = Break Sidebar Social Tag Category Structure!(SidebarSocialTagCategoryStructure=Get TagCategoryStructureReference()) -> (
    TagSkewBoxInnerColorHovered_42_#4:Linear Color Structure
    TagSkewBoxOuterColor_41_#3
   )
   > Set TargetInnerColorReference(TargetInnerColorReference:Linear Color Structure=@n3) -> (Output_Get:Linear Color Structure)
   > Set TargetOuterColorReference(TargetOuterColorReference:Linear Color Structure=@n3.TagSkewBoxOuterColor_41_#3) -> (Output_Get:Linear Color Structure)
   n4 = Break Sidebar Social Tag Category Structure!(SidebarSocialTagCategoryStructure=Get DisabledTagCategoryStructureReference()) -> (TagSkewBoxInnerColor_48_#2:Linear Color Structure, TagSkewBoxOuterColor_41_#3)
   Set TargetInnerColorReference(execute=@n2.else, TargetInnerColorReference=@n4)
   > Set TargetOuterColorReference(TargetOuterColorReference=@n4.TagSkewBoxOuterColor_41_#3)
   n5 = Set Timer by Event(execute=@n1.then_3, Delegate=Create Event(), Time=Get StateChangeLerpTimeStep() -> (StateChangeLerpTimeStep:double), bLooping=true)
   > Set StateChangeLerpTimerHandle(StateChangeLerpTimerHandle:Timer Handle Structure=@n5) -> (Output_Get:Timer Handle Structure)
   > SetScrollingEnabled(self=Get Text_AddTagName(), bInIsScrollingEnabled=true)
  ]}
 FunctionOnUnhovered {nodes [
   FunctionOnUnhovered() {is=title K2Node_FunctionEntry}
   > n1 = Sequence()
   > StopAllAnimations()
   > PlayAnimation(InAnimation=Get Hovered())
   Clear and Invalidate Timer by Handle(execute=@n1.then_1, Handle=Get StateChangeLerpTimerHandle())
   > GetCurrentTagBorderColors()
   > Set CurrentTagContentColorReference(CurrentTagContentColorReference=Get TagContentColorBlack())
   > Set TargetTagContentColorReference(TargetTagContentColorReference=Get TagContentColorWhite())
   n2 = Branch(execute=@n1.then_2)
   n3 = Break Sidebar Social Tag Category Structure!(SidebarSocialTagCategoryStructure=Get TagCategoryStructureReference()) -> (TagSkewBoxInnerColor_48_#2)
   > Set TargetInnerColorReference(TargetInnerColorReference=@n3)
   > Set TargetOuterColorReference(TargetOuterColorReference=@n3)
   n4 = Break Sidebar Social Tag Category Structure!(SidebarSocialTagCategoryStructure=Get DisabledTagCategoryStructureReference()) -> (TagSkewBoxInnerColor_48_#2, TagSkewBoxOuterColor_41_#3:Linear Color Structure)
   Set TargetInnerColorReference(execute=@n2.else, TargetInnerColorReference=@n4)
   > Set TargetOuterColorReference(TargetOuterColorReference=@n4.TagSkewBoxOuterColor_41_#3)
   n5 = Set Timer by Event(execute=@n1.then_3, Delegate=Create Event(), Time=Get StateChangeLerpTimeStep(), bLooping=true)
   > Set StateChangeLerpTimerHandle(StateChangeLerpTimerHandle=@n5)
   > SetScrollingEnabled(self=Get Text_AddTagName(), bInIsScrollingEnabled=false)
  ]}
 FunctionSetCategoryCustomizationVisibility {nodes [
   n1 = FunctionSetCategoryCustomizationVisibility!() -> (then, CategoryCustomizationVisibility:ESlateVisibility Enum=Visible) {is=title K2Node_FunctionEntry}
   > SetVisibility(self=Get Border_AddTagBorder(), InVisibility=@n1)
  ]}
 GetCurrentTagBorderColors {nodes [
   GetCurrentTagBorderColors() {is=title K2Node_FunctionEntry}
   > n1 = Get()
   > n2 = Get Vector Parameter Value!(execute, self:Material Instance Dynamic Object Reference=@n1, ParameterName:Name=Get Name_InnerColor() -> (Name_InnerColor=None)) -> (then, ReturnValue:Linear Color Structure) {is=title K2Node_CallFunction}
   > Set CurrentInnerColorReference(CurrentInnerColorReference:Linear Color Structure=@n2) -> (Output_Get:Linear Color Structure)
   > n3 = Get Vector Parameter Value(self=@n1, ParameterName=Get Name_OuterColor() -> (Name_OuterColor=None))
   > Set CurrentOuterColorReference(CurrentOuterColorReference:Linear Color Structure=@n3) -> (Output_Get:Linear Color Structure)
  ]}
]
variables [
 CurrentInnerColorReference {type=LinearColor, default=<rgba 0 0 0 1>}
 CurrentOuterColorReference {type=LinearColor, default=<rgba 0 0 0 1>}
 CurrentTagContentColorReference {type=LinearColor, default=<rgba 0 0 0 0>}
 DisabledTagCategoryStructureReference {
  type=SidebarSocialTagCategoryStructure
  default {
   tagCategoryIconMaterial=null
   tagCategoryBaseColor {colorUseRule=UseColor_Specified, specifiedColor=<rgba 1 0 1 1>}
   tagSkewBoxInnerColor=<rgba 0 0 0 0>
   tagSkewBoxInnerColorHovered=<rgba 0 0 0 0>
   tagSkewBoxInnerColorSelected=<rgba 0 0 0 1>
   tagSkewBoxOuterColor=<rgba 0 0 0 0>
   tagSkewBoxOuterColorSelected=<rgba 0 0 0 0>
  }
 }
 DisabledTagOpacity {type=number, default=0.6}
 EnabledTagOpacity {type=number, default=1}
 Name_InnerColor {type=string, default=InnerColor}
 Name_OuterColor {type=string, default=OuterColor}
 Name_TagOpacity {type=string, default=Opacity}
 StateChangeLerpAlpha {type=number}
 StateChangeLerpTimerHandle {type=TimerHandle}
 StateChangeLerpTimeStep {type=number, default=0.05}
 TagBorderMaterialReference {type=MaterialInstanceDynamic}
 TagCategory_AddTag {type=string, default=AddTag}
 TagCategory_Disabled {type=string, default=Disabled}
 TagCategoryStructureReference {
  type=SidebarSocialTagCategoryStructure
  default {
   tagCategoryIconMaterial=null
   tagCategoryBaseColor {colorUseRule=UseColor_Specified, specifiedColor=<rgba 1 0 1 1>}
   tagSkewBoxInnerColor=<rgba 0 0 0 0>
   tagSkewBoxInnerColorHovered=<rgba 0 0 0 0>
   tagSkewBoxInnerColorSelected=<rgba 0 0 0 1>
   tagSkewBoxOuterColor=<rgba 0 0 0 0>
   tagSkewBoxOuterColorSelected=<rgba 0 0 0 0>
  }
 }
 TagContentColorBlack {type=LinearColor, default=<rgba 0 0 0 1>}
 TagContentColorWhite {type=LinearColor, default=<rgba 1 1 1 0.9>}
 TargetInnerColorReference {type=LinearColor, default=<rgba 0 0 0 1>}
 TargetOuterColorReference {type=LinearColor, default=<rgba 0 0 0 1>}
 TargetTagContentColorReference {type=LinearColor, default=<rgba 0 0 0 0>}
]
widgets {
 info {inheritedWidgetCount=1, parentClass=&CommonButtonBase, rootWidgetClass=&Overlay, widgetCount=10}
 widgets [
  Overlay_0 {widgetClassPath=&Overlay, widget=Overlay_0}
  Switcher_Tag {
   widgetClassPath=&CommonVisibilitySwitcher
   bIsVariable=true
   parent=Overlay_0
   widget=Switcher_Tag
   slot {class=OverlaySlot, properties {horizontalAlignment=HAlign_Fill, verticalAlignment=VAlign_Fill}}
  }
  Border_AddTagBorder {
   widgetClassPath=&Border
   bIsVariable=true
   parent=Switcher_Tag
   widget=Border_AddTagBorder
   properties {
    background {
     bIsDynamicallyLoaded=false
     drawAs=Image
     imageType=NoImage
     mirroring=NoMirror
     resourceName=None
     resourceObject=/Game/UI/Sidebar/Tags/Materials/BoxSkew_SocialTag
     tiling=NoTile
     imageSize=<32 32>
     margin=<ltrb 0 0 0 0>
     outlineSettings {
      bUseBrushTransparency=false
      roundingType=HalfHeightRadius
      width=0
      color {colorUseRule=UseColor_Specified, specifiedColor=<rgba 0 0 0 0>}
      cornerRadii=<0 0 0 0>
     }
     tintColor {colorUseRule=UseColor_Specified, specifiedColor=<rgba 1 1 1 1>}
     uVRegion {bIsValid=false, max=<0 0>, min=<0 0>}
    }
    padding=<ltrb 8 6 12 6>
   }
   slot {class=CommonVisibilitySwitcherSlot}
  }
  HBox_AddTag {
   widgetClassPath=&HorizontalBox
   parent=Border_AddTagBorder
   widget=HBox_AddTag
   slot {class=BorderSlot, properties {padding=<ltrb 8 6 12 6>}}
  }
  Text_AddTagName {
   widgetClassPath=&CommonTextBlock
   bIsVariable=true
   parent=HBox_AddTag
   widget=Text_AddTagName
   properties {
    bIsScrollingEnabled=false
    mobileFontSizeMultiplier=1.25
    scrollStyle=&TextScrollStyle-Base_C
    text=+Add Tag
    textOverflowPolicy=Ellipsis
    colorAndOpacity {colorUseRule=UseColor_Specified, specifiedColor=<rgba 1 1 1 0.9>}
    font {
     bForceMonospaced=false
     bMaterialIsStencil=false
     fontMaterial=null
     fontObject=&BurbankSmall
     letterSpacing=0
     monospacedWidth=1
     size=16
     skewAmount=0.15
     typefaceFontName=Black
     outlineSettings {
      bApplyOutlineToDropShadows=false
      bMiteredCorners=false
      bSeparateFillAlpha=false
      outlineMaterial=null
      outlineSize=0
      outlineColor=<rgba 0 0 0 1>
     }
    }
    margin=<ltrb 0 0 10 0>
    shadowOffset=<0 0>
   }
   slot {class=HorizontalBoxSlot, properties {verticalAlignment=VAlign_Center, padding=<ltrb 10 0 0 0>, size {sizeRule=Fill, value=1}}}
  }
  SBox_AddTagInput {
   widgetClassPath=&SizeBox
   parent=HBox_AddTag
   widget=SBox_AddTagInput
   properties {bOverride_HeightOverride=true, bOverride_WidthOverride=true, heightOverride=26, widthOverride=26}
   slot {class=HorizontalBoxSlot, properties {horizontalAlignment=HAlign_Right, verticalAlignment=VAlign_Center, padding=<ltrb 5 0 0 0>}}
  }
  InputActionWidget {
   widgetClassPath=&CommonActionWidget
   bInherited=true
   parent=SBox_AddTagInput
   widget=InputActionWidget
   properties {
    visibility=SelfHitTestInvisible
    inputActions [{dataTable=&InputActionDataTable, rowName=Select}]
    progressMaterialBrush {
     bIsDynamicallyLoaded=false
     drawAs=Image
     imageType=NoImage
     mirroring=NoMirror
     resourceName=None
     resourceObject=null
     tiling=NoTile
     imageSize=<16 16>
     margin=<ltrb 0 0 0 0>
     outlineSettings {
      bUseBrushTransparency=false
      roundingType=HalfHeightRadius
      width=0
      color {colorUseRule=UseColor_Specified, specifiedColor=<rgba 0 0 0 0>}
      cornerRadii=<0 0 0 0>
     }
     tintColor {colorUseRule=UseColor_Specified, specifiedColor=<rgba 1 1 1 1>}
     uVRegion {bIsValid=false, max=<0 0>, min=<0 0>}
    }
   }
   slot {class=SizeBoxSlot}
  }
  SidebarSocialAddingTagWidget {
   widgetClassPath=&/SidebarSocialAddingTagWidget
   bIsVariable=true
   parent=Switcher_Tag
   widget=SidebarSocialAddingTagWidget
   slot {class=CommonVisibilitySwitcherSlot}
  }
  CommonHardwareVisibilityBorder_0 {
   widgetClassPath=&CommonHardwareVisibilityBorder
   parent=Overlay_0
   widget=CommonHardwareVisibilityBorder_0
   properties {
    visibleType=Visible
    visibilityQuery {
     autoDescription=" ANY( Input.Touch )"
     tokenStreamVersion=0
     userDescription=""
     queryTokenStream [0,1,1,1,0]
     tagDictionary [{tagName=Input.Touch}]
    }
   }
   slot {class=OverlaySlot, properties {horizontalAlignment=HAlign_Fill, verticalAlignment=VAlign_Fill, padding=<ltrb 0 -20 0 -20>}}
  }
  Overlay_hitzone_mobile {
   widgetClassPath=&Overlay
   parent=CommonHardwareVisibilityBorder_0
   widget=Overlay_hitzone_mobile
   properties {visibility=Visible}
   slot {class=BorderSlot, properties {padding=<ltrb 0 0 0 0>}}
  }
 ]
}
