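# Sidebar social tag chip (SocialTagWidget child) showing a category icon, skewed material border, scrolling name, check mark and select prompt.
# SetTag looks up category and Disabled color structs; the Highlighted flag picks between them.
# Hover drives border material parameters, content colors and icon-button animations, partly via a looping-timer lerp using an animated alpha.
# Selection can switch to an "adding tag" widget; mobile enables scrolling and shrinks the icon.
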WidgetBlueprint /Game/UI/Sidebar/Tags/SidebarSocialTagWidget
animations [
 StateChange {
  binding_count=1
  length_seconds=0.0996
  object=&self:StateChange
  sequence {bindings [SidebarSocialTagWidget_C_143 {
     id=#1
     possessed_class=SidebarSocialTagWidget_C
     tracks [{
       type=MovieSceneFloatTrack
       display_name=State Change Lerp Alpha
       sections [{type=MovieSceneFloatSection, channels [None {type=MovieSceneScriptingFloatChannel, keys [[0,0],[1.9923,1]]}]}]
      }]
    }]}
 }
 BorderHover {
  binding_count=3
  length_seconds=0.1
  object=&self:BorderHover
  sequence {bindings [
    Text_TagName {
     id=#2
     possessed_class=CommonTextBlock
     tracks [{
       type=MovieSceneColorTrack
       display_name=Color and Opacity
       sections [{
         type=MovieSceneColorSection
         end_frame=2
         start_frame=0
         channels [
          Color.R {type=MovieSceneScriptingFloatChannel, keys [[0,1],[2,0]]}
          Color.G {type=MovieSceneScriptingFloatChannel, keys [[0,1],[2,0]]}
          Color.B {type=MovieSceneScriptingFloatChannel, keys [[0,1],[2,0]]}
          Color.A {type=MovieSceneScriptingFloatChannel, const [0.9,1]}
         ]
        }]
      }]
    }
    Image_CheckMark {
     id=#3
     possessed_class=Image
     tracks [{
       type=MovieSceneColorTrack
       display_name=Color and Opacity
       sections [{
         type=MovieSceneColorSection
         end_frame=2
         start_frame=0
         channels [
          Color.R {type=MovieSceneScriptingFloatChannel, keys [[0,1],[2,0]]}
          Color.G {type=MovieSceneScriptingFloatChannel, keys [[0,1],[2,0]]}
          Color.B {type=MovieSceneScriptingFloatChannel, keys [[0,1],[2,0]]}
          Color.A {type=MovieSceneScriptingFloatChannel, keys [[0,0.9],[2,1]]}
         ]
        }]
      }]
    }
    Border_TagBorder {
     id=#4
     possessed_class=Border
     tracks [{
       type=MovieSceneWidgetMaterialTrack
       sections [{type=MovieSceneParameterSection, end_frame=2, start_frame=0, channels [IsHovered {type=MovieSceneScriptingFloatChannel, keys [[0,0],[2,1]]}]}]
      }]
    }
   ]}
 }
 BorderUnhover {
  binding_count=3
  length_seconds=0.1
  object=&self:BorderUnhover
  sequence {bindings [
    Image_CheckMark {
     id=#5
     possessed_class=Image
     tracks [{
       type=MovieSceneColorTrack
       display_name=Color and Opacity
       sections [{
         type=MovieSceneColorSection
         end_frame=2
         start_frame=0
         channels [
          Color.R {type=MovieSceneScriptingFloatChannel, keys [[0,0],[2,1]]}
          Color.G {type=MovieSceneScriptingFloatChannel, keys [[0,0],[2,1]]}
          Color.B {type=MovieSceneScriptingFloatChannel, keys [[0,0],[2,1]]}
          Color.A {type=MovieSceneScriptingFloatChannel, keys [[0,1],[2,0.9]]}
         ]
        }]
      }]
    }
    Text_TagName {
     id=#6
     possessed_class=CommonTextBlock
     tracks [{
       type=MovieSceneColorTrack
       display_name=Color and Opacity
       sections [{
         type=MovieSceneColorSection
         end_frame=2
         start_frame=0
         channels [
          Color.R {type=MovieSceneScriptingFloatChannel, keys [[0,0],[2,1]]}
          Color.G {type=MovieSceneScriptingFloatChannel, keys [[0,0],[2,1]]}
          Color.B {type=MovieSceneScriptingFloatChannel, keys [[0,0],[2,1]]}
          Color.A {type=MovieSceneScriptingFloatChannel, keys [[0,1],[2,0.9]]}
         ]
        }]
      }]
    }
    Border_TagBorder {
     id=#7
     possessed_class=Border
     tracks [{
       type=MovieSceneWidgetMaterialTrack
       sections [{type=MovieSceneParameterSection, end_frame=2, start_frame=0, channels [IsHovered {type=MovieSceneScriptingFloatChannel, keys [[0,1],[2,0]]}]}]
      }]
    }
   ]}
 }
]
graphs [
 EventGraph {nodes [
   "On Input Icon Updated (InputActionWidget)"() {is=title K2Node_ComponentBoundEvent}
   > n1 = GetIconDynamicMaterial!(execute, self:Common Action Widget Object Reference=Get InputActionWidget() -> (InputActionWidget:Common Action Widget Object Reference)) -> (then, ReturnValue:Material Instance Dynamic Object Reference)
   > SetScalarParameterValue(self=@n1, ParameterName=Use Dark, Value=1)
   > SetScalarParameterValue(self=@n1, ParameterName=Outline Opacity, Value=0)
   InterpolateTagColors() {is=title K2Node_CustomEvent}
   n2 = Get StateChangeLerpAlpha() -> (StateChangeLerpAlpha:double)
   > InterpolateTagContentColors!(
    execute
    TagName:Text Object Reference=Get Text_TagName() -> (Text_TagName:Common Text Object Reference)
    CheckMark:Image Object Reference=Get Image_CheckMark() -> (Image_CheckMark:Image Object Reference)
    CurrentTagContentColor:Linear Color Structure=Get CurrentTagContentColorReference() -> (CurrentTagContentColorReference:Linear Color Structure)
    TargetTagContentColor:Linear Color Structure=Get TargetTagContentColorReference() -> (TargetTagContentColorReference:Linear Color Structure)
    HoveredTagContentLerpAlpha:double=@n2
   ) -> (then)
   > n3 = Get!(execute:Exec) -> (then:Exec, else:Exec, TagBorderMaterialReference:Material Instance Dynamic Object Reference)
   > InterpolateTagBorderColors!(
    execute
    TagBorderMaterial:Material Instance Dynamic Object Reference=@n3
    NameInnerColor:Name=None
    CurrentInnerColor:Linear Color Structure
    TargetInnerColor:Linear Color Structure
    NameOuterColor:Name=Get Name_OuterColor() -> (Name_OuterColor:Name=None)
    CurrentOuterColor:Linear Color Structure=Get CurrentOuterColorReference() -> (CurrentOuterColorReference:Linear Color Structure)
    TargetOuterColor:Linear Color Structure=Get TargetOuterColorReference() -> (TargetOuterColorReference:Linear Color Structure)
    HoveredBorderColorLerpAlpha:double=@n2
    ColorLerpTimeHandle:"Timer Handle Structure (by ref)"=Get StateChangeLerpTimerHandle() -> (StateChangeLerpTimerHandle:Timer Handle Structure)
    WorldContext:Object Reference=Self-Reference()
   ) -> (then)
   ClearLerpColorTimer!(
    execute=@n3.else
    ColorLerpAlpha:double=Get StateChangeLerpAlpha()
    ColorLerpTimeHandle:"Timer Handle Structure (by ref)"=Get StateChangeLerpTimerHandle()
   ) -> (then)
   n4 = Event SetFriendsCellColor() {is=title K2Node_CustomEvent}
   n5 = Event SetHighlightState!() -> (OutputDelegate, then, bHighlighted:bool) {is=title K2Node_Event}
   Set Highlighted(execute=[@n4.then, @n5.then], Highlighted:bool=@n5.bHighlighted) -> (Output_Get:bool)
   > n6 = FindParentWidgetOfType!(execute, StartingWidget:Widget Object Reference=Self-Reference(), Type:Widget Class Reference=&CommonButtonLegacy) -> (then, ReturnValue:Common Button Legacy Object Reference)
   > Is Valid(InputObject=@n6)
   n7 = Break Sidebar Social Tag Category Structure!(SidebarSocialTagCategoryStructure:"Sidebar Social Tag Category Structure Structure (by ref)"=Get TagCategoryStructureReference() -> (TagCategoryStructureReference:Sidebar Social Tag Category Structure Structure)) -> (
    TagSkewBoxInnerColor_48_#8:Linear Color Structure
    TagSkewBoxInnerColorSelected_45_#9:Linear Color Structure
    TagSkewBoxOuterColor_41_#10:Linear Color Structure
    TagSkewBoxOuterColorSelected_47_#11:Linear Color Structure
   ) {is=title K2Node_BreakStruct}
   > FunctionSetTagBorderColors!(
    execute
    self
    InnerColor:Linear Color Structure=@n7
    OuterColor:Linear Color Structure=@n7.TagSkewBoxOuterColor_41_#10
   ) -> (then)
   n8 = IsHovered(self=@n6)
   n9 = Select(
    Option 0:Linear Color Structure=Select(
     Option 0=Get TagContentColorWhiteDisabled() -> (TagContentColorWhiteDisabled:Linear Color Structure)
     Option 1=Get TagContentColorBlackDisabled() -> (TagContentColorBlackDisabled:Linear Color Structure)
     Index=@n8
    )
    Option 1:Linear Color Structure=Get TagContentColorWhite() -> (TagContentColorWhite:Linear Color Structure)
    Index=Get Highlighted() -> (Highlighted:bool)
   ) -> (ReturnValue:Linear Color Structure)
   > SetColorAndOpacity(self=Get Text_TagName(), InColorAndOpacity=Make Slate Color(SpecifiedColor=@n9))
   > SetColorAndOpacity(self:Image Object Reference=Get Image_CheckMark(), InColorAndOpacity:Linear Color Structure=@n9)
   n10 = Get BP_SidebarCollectionScreenContainerTabButtonEditTags() -> (BP_SidebarCollectionScreenContainerTabButtonEditTags:BP Sidebar Collection Screen Container Tab Button Edit Tags Object Reference)
   n11 = Get Tag_DisabledHovered!(self:BP Sidebar Collection Screen Container Tab Button Edit Tags Object Reference=@n10) -> (Tag_DisabledHovered:Widget Animation Object Reference)
   n12 = Get Tag_DisabledUnhovered!(self:BP Sidebar Collection Screen Container Tab Button Edit Tags Object Reference=@n10) -> (Tag_DisabledUnhovered:Widget Animation Object Reference)
   > PlayAnimation(
    self=@n10
    InAnimation=Select(
     Option 0:Widget Animation Object Reference=Select(Option 0=@n12, Option 1=@n11, Index=@n8)
     Option 1:Widget Animation Object Reference=Get Tag_Enabled!(self:BP Sidebar Collection Screen Container Tab Button Edit Tags Object Reference=@n10) -> (Tag_Enabled:Widget Animation Object Reference)
     Index=Get Highlighted()
    ) -> (ReturnValue:Widget Animation Object Reference)
   )
   Event Construct()
   > Branch(Condition=IsMobileGame())
   > SetScrollingEnabled!(execute, self:Common Text Object Reference=Get Text_TagName(), bInIsScrollingEnabled:bool=true) -> (then)
   > SetUserSpecifiedScale!(
    execute
    self:Scale Box Object Reference=Get ScaleBox_TagCategoryIcon() -> (ScaleBox_TagCategoryIcon:Scale Box Object Reference)
    InUserSpecifiedScale:float=0.8
   ) -> (then)
   Event SetFullColor() {is=title K2Node_Event}
   > OnTagListEntryHovered()
   n13 = Event SetTag!() -> (OutputDelegate, then, InTagName:"Text (by ref)", InTagCategory:ESocialTagCategory Enum=GameModes, bInIsSystemTag:bool) {is=title K2Node_Event}
   > FunctionSetTag!(
    execute
    self
    InTagName:Text=@n13.InTagName
    InTagCategory:ESocialTagCategory Enum=@n13.InTagCategory
    InIsSystemTag:bool=@n13.bInIsSystemTag
   ) -> (then)
   n14 = Event SetCategoryCustomizationVisibility!() -> (OutputDelegate, then, CategoryCustomizationVisibility:ESlateVisibility Enum=Visible) {is=title K2Node_Event}
   > FunctionSetCategoryCustomizationVisibility!(execute, self, CategoryCustomizationVisibility:ESlateVisibility Enum=@n14.CategoryCustomizationVisibility) -> (then)
   Event On Hovered()
   > SetTagHoverState!(execute, self, IsUserSocialTag:bool=true, IsHovering:bool=true) -> (then)
   Event On Unhovered()
   > SetTagHoverState(IsUserSocialTag=true, IsHovering=false)
   Event On Deselected()
   n15 = Get bShowAddingWhenSelected() -> (bShowAddingWhenSelected:bool)
   > Branch(Condition=@n15)
   n16 = Get Switcher_SelectedState() -> (Switcher_SelectedState:Common Visibility Switcher Object Reference)
   > SetActiveWidget(self=@n16, Widget=Get GridPanel_TagWidget() -> (GridPanel_TagWidget:Grid Panel Object Reference))
   Event On Selected()
   > Branch(Condition=@n15)
   > SetActiveWidget(self=@n16, Widget=Get SidebarSocialAddingTagWidget() -> (SidebarSocialAddingTagWidget:Sidebar Social Adding Tag Widget Object Reference))
  ]}
 FunctionOnHovered {nodes [
   n1 = Set Timer by Event(Delegate=Create Event(), Time=Get StateChangeLerpTimeStep() -> (StateChangeLerpTimeStep:double), bLooping=true)
   > Set StateChangeLerpTimerHandle(StateChangeLerpTimerHandle:Timer Handle Structure=@n1) -> (Output_Get:Timer Handle Structure)
   Clear and Invalidate Timer by Handle(Handle=Get StateChangeLerpTimerHandle())
   > GetCurrentTagBorderColors()
   > Set CurrentTagContentColorReference(CurrentTagContentColorReference:Linear Color Structure=Get ColorAndOpacity!(self:Image Object Reference=Get Image_CheckMark()) -> (ColorAndOpacity:Linear Color Structure)) -> (Output_Get:Linear Color Structure)
   > Set TargetTagContentColorReference(TargetTagContentColorReference:Linear Color Structure=Get TagContentColorBlack() -> (TagContentColorBlack:Linear Color Structure)) -> (Output_Get:Linear Color Structure)
   FunctionOnHovered() {is=title K2Node_FunctionEntry}
   > n2 = Sequence!(execute) -> (then_0, then_1, then_2:Exec, then_3:Exec)
   Branch(execute=@n2.then_3, Condition=IsAnimationPlaying(InAnimation=Get StateChange() -> (StateChange:Widget Animation Object Reference)))
   > SetScrollingEnabled(self=Get Text_TagName(), bInIsScrollingEnabled=true)
   n3 = Get BP_SidebarCollectionScreenContainerTabButtonEditTags()
   PlayAnimation(
    execute=@n2.then_0
    self=@n3
    InAnimation=Get Tag_Hovered!(self:BP Sidebar Collection Screen Container Tab Button Edit Tags Object Reference=@n3) -> (Tag_Hovered:Widget Animation Object Reference)
   )
   > PlayAnimation(InAnimation=Get StateChange())
   n4 = Branch(Condition=Get Highlighted())
   n5 = Break Sidebar Social Tag Category Structure!(SidebarSocialTagCategoryStructure=Get TagCategoryStructureReference()) -> (
    TagSkewBoxInnerColorHovered_42_#12:Linear Color Structure
    TagSkewBoxOuterColorSelected_47_#11
   )
   > Set TargetInnerColorReference(TargetInnerColorReference:Linear Color Structure=@n5) -> (Output_Get:Linear Color Structure)
   > Set TargetOuterColorReference(TargetOuterColorReference:Linear Color Structure=@n5.TagSkewBoxOuterColorSelected_47_#11) -> (Output_Get:Linear Color Structure)
   n6 = Break Sidebar Social Tag Category Structure!(SidebarSocialTagCategoryStructure=Get DisabledTagCategoryStructureReference() -> (DisabledTagCategoryStructureReference:Sidebar Social Tag Category Structure Structure)) -> (
    TagSkewBoxInnerColor_48_#8:Linear Color Structure
    TagSkewBoxOuterColor_41_#10:Linear Color Structure
   )
   Set TargetInnerColorReference(execute=@n4.else, TargetInnerColorReference=@n6)
   > Set TargetOuterColorReference(TargetOuterColorReference=@n6.TagSkewBoxOuterColor_41_#10)
   PlayAnimation(InAnimation=Get BorderHover() -> (BorderHover:Widget Animation Object Reference))
  ]}
 FunctionOnUnhovered {nodes [
   n1 = Set Timer by Event(Delegate=Create Event(), Time=Get StateChangeLerpTimeStep(), bLooping=true)
   > Set StateChangeLerpTimerHandle(StateChangeLerpTimerHandle=@n1)
   Clear and Invalidate Timer by Handle(Handle=Get StateChangeLerpTimerHandle())
   > GetCurrentTagBorderColors()
   > Set CurrentTagContentColorReference(CurrentTagContentColorReference=Get ColorAndOpacity(self=Get Image_CheckMark()))
   > Set TargetTagContentColorReference(TargetTagContentColorReference=Get TagContentColorWhite())
   FunctionOnUnhovered() {is=title K2Node_FunctionEntry}
   > n2 = Sequence()
   Branch(execute=@n2.then_3, Condition=IsAnimationPlaying(InAnimation=Get StateChange()))
   > SetScrollingEnabled(self=Get Text_TagName(), bInIsScrollingEnabled=false)
   n3 = Get BP_SidebarCollectionScreenContainerTabButtonEditTags()
   PlayAnimation(
    execute=@n2.then_0
    self=@n3
    InAnimation=Get Tag_Unhovered!(self:BP Sidebar Collection Screen Container Tab Button Edit Tags Object Reference=@n3) -> (Tag_Unhovered:Widget Animation Object Reference)
   )
   > PlayAnimation(InAnimation=Get StateChange())
   > PlayAnimation(InAnimation=Get BorderUnhover() -> (BorderUnhover:Widget Animation Object Reference))
   n4 = Branch(Condition=Get Highlighted())
   n5 = Break Sidebar Social Tag Category Structure(SidebarSocialTagCategoryStructure=Get TagCategoryStructureReference())
   > Set TargetInnerColorReference(TargetInnerColorReference=@n5)
   > Set TargetOuterColorReference(TargetOuterColorReference=@n5.TagSkewBoxOuterColor_41_#10)
   n6 = Break Sidebar Social Tag Category Structure(SidebarSocialTagCategoryStructure=Get DisabledTagCategoryStructureReference())
   Set TargetInnerColorReference(execute=@n4.else, TargetInnerColorReference=@n6)
   > Set TargetOuterColorReference(TargetOuterColorReference=@n6.TagSkewBoxOuterColor_41_#10)
  ]}
 FunctionSetCategoryCustomizationVisibility {nodes [
   n1 = FunctionSetCategoryCustomizationVisibility!() -> (then, CategoryCustomizationVisibility:ESlateVisibility Enum=Visible) {is=title K2Node_FunctionEntry}
   > SetVisibility(self=[Get Spacer_TagCategoryIcon() -> (Spacer_TagCategoryIcon:Spacer Object Reference), Get ScaleBox_TagCategoryIcon()], InVisibility=@n1)
   n2 = Get BP_SidebarCollectionScreenContainerTabButtonEditTags()
   > PlayAnimation(
    self=@n2
    InAnimation=Get Tag_Reset!(self:BP Sidebar Collection Screen Container Tab Button Edit Tags Object Reference=@n2) -> (Tag_Reset:Widget Animation Object Reference)
   )
  ]}
 FunctionSetTag {nodes [
   n1 = FunctionSetTag!() -> (then, InTagName:Text, InTagCategory:ESocialTagCategory Enum=GameModes, InIsSystemTag:bool=~) {is=title K2Node_FunctionEntry}
   > n2 = Sequence()
   n3 = Get TagContentColorWhite()
   SetColorAndOpacity(
    execute=@n2.then_2
    self:Text Object Reference=Get Text_TagName()
    InColorAndOpacity:Slate Color Structure=Make Slate Color(SpecifiedColor=@n3)
   )
   > SetColorAndOpacity(self:Image Object Reference=Get Image_CheckMark(), InColorAndOpacity:Linear Color Structure=@n3)
   n4 = Break Sidebar Social Tag Category Structure!(SidebarSocialTagCategoryStructure=Get TagCategoryStructureReference()) -> (
    TagSkewBoxInnerColor_48_#8
    TagSkewBoxOuterColor_41_#10
    TagCategoryIconMaterial_7_#13:Material Instance Object Reference
   )
   > n5 = Is Valid(InputObject=@n4.TagCategoryIconMaterial_7_#13)
   SetVisibility(execute=@n5.Is Not Valid, self=[Get Spacer_TagCategoryIcon(), Get ScaleBox_TagCategoryIcon()], InVisibility=Collapsed)
   Set Icon!(
    execute=@n5.Is Valid
    self:Athena Collection Screen Container Tab Button Object Reference=Get BP_SidebarCollectionScreenContainerTabButtonEditTags()
    LazyObject:Soft Object Reference=Make Soft Reference!(Input:Material Instance Object Reference=@n4.TagCategoryIconMaterial_7_#13) -> (Output:Material Instance Soft Object Reference)
   ) -> (then) {is=title K2Node_CallFunction}
   Set TargetInnerColorReference(execute=@n2.then_3, TargetInnerColorReference=@n4)
   > Set TargetOuterColorReference(TargetOuterColorReference=@n4.TagSkewBoxOuterColor_41_#10)
   > FunctionSetTagBorderColors(InnerColor=@n4, OuterColor=@n4.TagSkewBoxOuterColor_41_#10)
   "SetText (Text)"(execute=@n2.then_0, self=Get Text_TagName(), InText=@n1)
   > SetScrollingEnabled(self=Get Text_TagName(), bInIsScrollingEnabled=false)
   > n6 = GetDynamicMaterial(self:Border Object Reference=Get Border_TagBorder() -> (Border_TagBorder:Border Object Reference))
   > Is Valid(InputObject=@n6)
   > Set TagBorderMaterialReference(TagBorderMaterialReference:Material Instance Dynamic Object Reference=@n6) -> (Output_Get:Material Instance Dynamic Object Reference)
   n7 = Get Sidebar Social Tag Category Structure!(execute=@n2.then_1, TagCategory:String=Enum to String!(Enumerator:ESocialTagCategory Enum=@n1.InTagCategory) -> (ReturnValue:String)) -> (then, SidebarSocialTagCategoryStructure:Sidebar Social Tag Category Structure Structure) {is=title K2Node_CallFunction}
   > Set TagCategoryStructureReference(TagCategoryStructureReference:Sidebar Social Tag Category Structure Structure=@n7) -> (Output_Get:Sidebar Social Tag Category Structure Structure)
   > n8 = Get Sidebar Social Tag Category Structure(TagCategory=Get TagCategory_Disabled() -> (TagCategory_Disabled:String))
   > Set DisabledTagCategoryStructureReference(DisabledTagCategoryStructureReference:Sidebar Social Tag Category Structure Structure=@n8) -> (Output_Get:Sidebar Social Tag Category Structure Structure)
  ]}
 FunctionSetTagBorderColors {nodes [
   n1 = FunctionSetTagBorderColors!() -> (then, InnerColor:Linear Color Structure, OuterColor:Linear Color Structure) {is=title K2Node_FunctionEntry}
   > n2 = GetDynamicMaterial(self=Get Border_TagBorder())
   > Is Valid(InputObject=@n2)
   n3 = Break Sidebar Social Tag Category Structure!(SidebarSocialTagCategoryStructure=Get DisabledTagCategoryStructureReference()) -> (TagSkewBoxInnerColor_48_#8, TagSkewBoxOuterColor_41_#10)
   n4 = Get Highlighted()
   > Set TargetInnerColorReference(TargetInnerColorReference=SelectColor(A=@n1, B=@n3, bPickA=@n4))
   > Set TargetOuterColorReference(TargetOuterColorReference=SelectColor(A=@n1.OuterColor, B=@n3.TagSkewBoxOuterColor_41_#10, bPickA=@n4))
   > Set TargetTagContentColorReference(TargetTagContentColorReference=SelectColor(A=Get TagContentColorWhite(), B=Get TagContentColorWhiteDisabled(), bPickA=@n4))
   > SetVectorParameterValue(
    self=@n2
    ParameterName=Get Name_InnerColor() -> (Name_InnerColor:Name=None)
    Value=Get TargetInnerColorReference() -> (TargetInnerColorReference:Linear Color Structure)
   )
   > SetVectorParameterValue(self=@n2, ParameterName=Get Name_OuterColor() -> (Name_OuterColor=None), Value=Get TargetOuterColorReference())
  ]}
 GetCurrentTagBorderColors {nodes [
   GetCurrentTagBorderColors() {is=title K2Node_FunctionEntry}
   > n1 = Get()
   > n2 = Get Vector Parameter Value!(execute, self:Material Instance Dynamic Object Reference=@n1, ParameterName:Name=Get Name_InnerColor() -> (Name_InnerColor=None)) -> (then, ReturnValue:Linear Color Structure) {is=title K2Node_CallFunction}
   > Set CurrentInnerColorReference(CurrentInnerColorReference:Linear Color Structure=@n2) -> (Output_Get:Linear Color Structure)
   > n3 = Get Vector Parameter Value(self=@n1, ParameterName=Get Name_OuterColor() -> (Name_OuterColor=None))
   > Set CurrentOuterColorReference(CurrentOuterColorReference:Linear Color Structure=@n3) -> (Output_Get:Linear Color Structure)
  ]}
 OnTagListEntryHovered {nodes [
   OnTagListEntryHovered() {is=title K2Node_FunctionEntry}
   > n1 = Sequence()
   Clear and Invalidate Timer by Handle(execute=@n1.then_1, Handle=Get StateChangeLerpTimerHandle())
   > GetCurrentTagBorderColors() {is=title K2Node_CallFunction}
   > Set CurrentTagContentColorReference(CurrentTagContentColorReference=Get ColorAndOpacity(self=Get Image_CheckMark()))
   n2 = Branch(execute=@n1.then_2, Condition=Get Highlighted())
   n3 = Break Sidebar Social Tag Category Structure!(SidebarSocialTagCategoryStructure=Get TagCategoryStructureReference()) -> (
    TagSkewBoxInnerColorSelected_45_#9:Linear Color Structure
    TagSkewBoxOuterColorSelected_47_#11:Linear Color Structure
   )
   > Set TargetInnerColorReference(TargetInnerColorReference=@n3)
   > Set TargetOuterColorReference(TargetOuterColorReference=@n3.TagSkewBoxOuterColorSelected_47_#11)
   > Set TargetTagContentColorReference(TargetTagContentColorReference=Get TagContentColorWhite())
   n4 = Break Sidebar Social Tag Category Structure(SidebarSocialTagCategoryStructure=Get DisabledTagCategoryStructureReference())
   Set TargetInnerColorReference(execute=@n2.else, TargetInnerColorReference=@n4)
   > Set TargetOuterColorReference(TargetOuterColorReference=@n4.TagSkewBoxOuterColorSelected_47_#11)
   > Set TargetTagContentColorReference(TargetTagContentColorReference=Get TagContentColorBlackDisabled())
   Branch(execute=@n1.then_3, Condition=IsAnimationPlaying(InAnimation=Get StateChange()))
   > n5 = Set Timer by Event(Delegate=Create Event(), Time=Get StateChangeLerpTimeStep(), bLooping=true)
   > Set StateChangeLerpTimerHandle(StateChangeLerpTimerHandle=@n5)
   > SetScrollingEnabled(self=Get Text_TagName(), bInIsScrollingEnabled=true)
   n6 = Branch(execute=@n1.then_0, Condition=Get Highlighted())
   PlayAnimation(execute=[@n8.then, @n6.then], InAnimation=Get StateChange())
   n7 = Get BP_SidebarCollectionScreenContainerTabButtonEditTags()
   n8 = PlayAnimation(
    execute=@n6.else
    self=@n7
    InAnimation=Get Tag_DisabledHover!(self:BP Sidebar Collection Screen Container Tab Button Edit Tags Object Reference=@n7) -> (Tag_DisabledHover:Widget Animation Object Reference)
   )
  ]}
 OnTagListEntryUnhovered {nodes [
   OnTagListEntryUnhovered() {is=title K2Node_FunctionEntry}
   > SetUnhoverState!(execute, self, bPlayAnimation:bool=true) -> (then)
  ]}
 SetTagHoverState {nodes [
   n1 = SetTagHoverState!() -> (then, IsUserSocialTag:bool=~, IsHovering:bool=~) {is=title K2Node_FunctionEntry}
   > n2 = Branch(Condition=@n1)
   > SetVectorParameterValue(
    self=Get TagBorderMaterialReference() -> (TagBorderMaterialReference:Material Instance Dynamic Object Reference)
    ParameterName=Get Name_EndingOuterColor() -> (Name_EndingOuterColor:Name=None)
    Value=Break Sidebar Social Tag Category Structure!(SidebarSocialTagCategoryStructure=Get TagCategoryStructureReference()) -> (TagSkewBoxOuterColor_41_#10:Linear Color Structure)
   )
   > n3 = Branch(Condition=@n1.IsHovering)
   n4 = Get BP_SidebarCollectionScreenContainerTabButtonEditTags()
   n5 = Get Tag_Hovered(self=@n4)
   > PlayAnimation(self=@n4, InAnimation=@n5)
   n6 = Get BorderHover()
   > PlayAnimation(InAnimation=@n6)
   > SetScrollingEnabled(self=Get Text_TagName(), bInIsScrollingEnabled=true)
   PlayAnimation(execute=@n3.else, InAnimation=@n5, PlayMode=Reverse)
   > PlayAnimation(InAnimation=@n6, PlayMode=Reverse)
   > SetScrollingEnabled(self=Get Text_TagName(), bInIsScrollingEnabled=false)
   n7 = Sequence!(execute=@n2.else) -> (then_0, then_1)
   > n8 = Branch(Condition=Get Highlighted())
   n9 = Break Sidebar Social Tag Category Structure!(SidebarSocialTagCategoryStructure=Get TagCategoryStructureReference()) -> (
    TagSkewBoxInnerColor_48_#8:Linear Color Structure
    TagSkewBoxInnerColorSelected_45_#9:Linear Color Structure
    TagSkewBoxOuterColor_41_#10
    TagSkewBoxOuterColorSelected_47_#11:Linear Color Structure
   )
   n10 = Get TagBorderMaterialReference()
   > SetVectorParameterValue(
    self=@n10
    ParameterName=Get Name_EndingOuterColor() -> (Name_EndingOuterColor=None)
    Value=@n9.TagSkewBoxOuterColorSelected_47_#11
   )
   > SetVectorParameterValue(
    self=@n10
    ParameterName=Get Name_EndingInnerColor() -> (Name_EndingInnerColor:Name=None)
    Value=@n9.TagSkewBoxInnerColorSelected_45_#9
   )
   n11 = Break Sidebar Social Tag Category Structure!(SidebarSocialTagCategoryStructure=Get DisabledTagCategoryStructureReference()) -> (TagSkewBoxInnerColor_48_#8, TagSkewBoxOuterColor_41_#10)
   SetVectorParameterValue(
    execute=@n8.else
    self=@n10
    ParameterName=Get Name_EndingOuterColor() -> (Name_EndingOuterColor=None)
    Value=@n11.TagSkewBoxOuterColor_41_#10
   )
   > SetVectorParameterValue(self=@n10, ParameterName=Get Name_EndingInnerColor() -> (Name_EndingInnerColor=None), Value=@n11)
   n12 = GetDynamicMaterial(execute=@n7.then_1, self=Get Border_TagBorder())
   > SetScalarParameterValue(self=@n12, ParameterName=IsHovered, Value=Select(Option 0:float=0, Option 1:float=1, Index=@n1.IsHovering) -> (ReturnValue:float))
   > SetScrollingEnabled(self=Get Text_TagName(), bInIsScrollingEnabled=Select(Option 0:bool=false, Option 1:bool=true, Index=@n1.IsHovering) -> (ReturnValue:bool))
   > Event SetFriendsCellColor() {is=title K2Node_CallFunction}
   Reroute Node()
   > n13 = GetDynamicMaterial(self=Get Border_TagBorder())
   > SetScalarParameterValue(self=@n13, ParameterName=IsHovered, Value=0)
   > SetScrollingEnabled(self=Get Text_TagName(), bInIsScrollingEnabled=false)
   > Call SetFriendsCellColor!(execute:Exec, self:Sidebar Social Tag Widget Object Reference) -> (then:Exec) {is=title K2Node_CallDelegate}
  ]}
 SetUnhoverState {nodes [
   SetUnhoverState!() -> (then, bPlayAnimation:bool=true) {is=title K2Node_FunctionEntry}
   > n1 = Sequence!(execute) -> (then_0, then_1, then_2:Exec, then_3:Exec)
   Clear and Invalidate Timer by Handle(execute=@n1.then_1, Handle=Get StateChangeLerpTimerHandle())
   > GetCurrentTagBorderColors()
   > Set CurrentTagContentColorReference(CurrentTagContentColorReference=Get ColorAndOpacity(self=Get Image_CheckMark()))
   n2 = Branch(execute=@n1.then_2, Condition=Get Highlighted())
   n3 = Break Sidebar Social Tag Category Structure(SidebarSocialTagCategoryStructure=Get TagCategoryStructureReference())
   > Set TargetInnerColorReference(TargetInnerColorReference=@n3)
   > Set TargetOuterColorReference(TargetOuterColorReference=@n3.TagSkewBoxOuterColor_41_#10)
   > Set TargetTagContentColorReference(TargetTagContentColorReference=Get TagContentColorWhite())
   n4 = Break Sidebar Social Tag Category Structure(SidebarSocialTagCategoryStructure=Get DisabledTagCategoryStructureReference())
   Set TargetInnerColorReference(execute=@n2.else, TargetInnerColorReference=@n4)
   > Set TargetOuterColorReference(TargetOuterColorReference=@n4.TagSkewBoxOuterColor_41_#10)
   > Set TargetTagContentColorReference(TargetTagContentColorReference=Get TagContentColorWhiteDisabled())
   n5 = Branch(execute=@n1.then_0, Condition=Get Highlighted())
   n6 = Get StateChange()
   PlayAnimation(
    execute=[@n9.then, @n5.then]
    InAnimation=@n6
    StartAtTime=SelectFloat(
     A=GetStartTime!(self:Widget Animation Object Reference=@n6) -> (ReturnValue:float)
     B=GetEndTime(self=@n6)
     bPickA=Get bPlayAnimation() -> (bPlayAnimation:bool)
    )
   )
   n7 = Get BP_SidebarCollectionScreenContainerTabButtonEditTags()
   n8 = Get Tag_DisabledUnhover!(self:BP Sidebar Collection Screen Container Tab Button Edit Tags Object Reference=@n7) -> (Tag_DisabledUnhover:Widget Animation Object Reference)
   n9 = PlayAnimation(
    execute=@n5.else
    self=@n7
    InAnimation=@n8
    StartAtTime=SelectFloat(A=GetStartTime(self=@n8), B=GetEndTime(self=@n8), bPickA=Get bPlayAnimation())
   )
   n10 = Branch(execute=@n1.then_3, Condition=Get bPlayAnimation())
   > Branch(Condition=IsAnimationPlaying(InAnimation=Get StateChange()))
   > n11 = Set Timer by Event(Delegate=Create Event(), Time=Get StateChangeLerpTimeStep(), bLooping=true)
   > n12 = Set StateChangeLerpTimerHandle(StateChangeLerpTimerHandle=@n11)
   n13 = Branch(execute=[@n14.then, @n12.then], Condition=IsMobileGame())
   SetScrollingEnabled(execute=@n13.else, self=Get Text_TagName(), bInIsScrollingEnabled=false)
   n14 = InterpolateTagColors(execute=@n10.else) {is=title K2Node_CallFunction}
  ]}
]
variables [
 CurrentInnerColorReference {type=LinearColor, default=<rgba 0 0 0 1>}
 CurrentOuterColorReference {type=LinearColor, default=<rgba 0 0 0 1>}
 CurrentTagContentColorReference {type=LinearColor, default=<rgba 0 0 0 1>}
 DisabledTagCategoryStructureReference {
  type=SidebarSocialTagCategoryStructure
  default {
   tagCategoryIconMaterial=null
   tagCategoryBaseColor {colorUseRule=UseColor_Specified, specifiedColor=<rgba 1 0 1 1>}
   tagSkewBoxInnerColor=<rgba 0 0 0 0>
   tagSkewBoxInnerColorHovered=<rgba 0 0 0 0>
   tagSkewBoxInnerColorSelected=<rgba 0 0 0 1>
   tagSkewBoxOuterColor=<rgba 0 0 0 0>
   tagSkewBoxOuterColorSelected=<rgba 0 0 0 0>
  }
 }
 DisabledTagOpacity {type=number, default=0.6}
 EnabledTagOpacity {type=number, default=1}
 Highlighted {type=boolean, default=true}
 Name_EndingInnerColor {type=string, default=EndingInnerColor}
 Name_EndingOuterColor {type=string, default=EndingOuterColor}
 Name_GradientEndColor {type=string, default=Gradient End Color}
 Name_GradientStartColor {type=string, default=Gradient Start Color}
 Name_IconColor {type=string, default=Icon Color}
 Name_InnerColor {type=string, default=InnerColor}
 Name_OuterColor {type=string, default=OuterColor}
 Name_TagOpacity {type=string, default=Opacity}
 StateChangeLerpAlpha {type=number}
 StateChangeLerpTimerHandle {type=TimerHandle}
 StateChangeLerpTimeStep {type=number, default=0.05}
 TagBorderMaterialReference {type=MaterialInstanceDynamic}
 TagCategory_Disabled {type=string, default=Disabled}
 TagCategoryIconMaterialReference {type=MaterialInstanceDynamic}
 TagCategoryStructureReference {
  type=SidebarSocialTagCategoryStructure
  default {
   tagCategoryIconMaterial=null
   tagCategoryBaseColor {colorUseRule=UseColor_Specified, specifiedColor=<rgba 1 0 1 1>}
   tagSkewBoxInnerColor=<rgba 0 0 0 0>
   tagSkewBoxInnerColorHovered=<rgba 0 0 0 0>
   tagSkewBoxInnerColorSelected=<rgba 0 0 0 1>
   tagSkewBoxOuterColor=<rgba 0 0 0 0>
   tagSkewBoxOuterColorSelected=<rgba 0 0 0 0>
  }
 }
 TagContentColorBlack {type=LinearColor, default=<rgba 0 0 0 1>}
 TagContentColorBlackDisabled {type=LinearColor, default=<rgba 0 0 0 0.6>}
 TagContentColorWhite {type=LinearColor, default=<rgba 1 1 1 0.9>}
 TagContentColorWhiteDisabled {type=LinearColor, default=<rgba 1 1 1 0.6>}
 TargetInnerColorReference {type=LinearColor, default=<rgba 0 0 0 0.5>}
 TargetOuterColorReference {type=LinearColor, default=<rgba 0 0 0 1>}
 TargetTagContentColorReference {type=LinearColor, default=<rgba 1 1 1 0.9>}
]
widgets {
 info {
  inheritedWidgetCount=2
  parentClass=/Script/FortniteUI.SocialTagWidget
  rootWidgetClass=&CommonVisibilitySwitcher
  widgetCount=12
 }
 widgets [
  Switcher_SelectedState {widgetClassPath=&CommonVisibilitySwitcher, bIsVariable=true, widget=Switcher_SelectedState}
  GridPanel_TagWidget {
   widgetClassPath=&GridPanel
   bIsVariable=true
   parent=Switcher_SelectedState
   widget=GridPanel_TagWidget
   properties {columnFill [0,1]}
   slot {class=CommonVisibilitySwitcherSlot}
  }
  ScaleBox_TagCategoryIcon {
   widgetClassPath=&ScaleBox
   bIsVariable=true
   parent=GridPanel_TagWidget
   widget=ScaleBox_TagCategoryIcon
   properties {stretch=UserSpecified, userSpecifiedScale=0.9}
   slot {class=GridSlot, properties {layer=1, nudge=<31 0>, padding=<ltrb -31 0 0 0>}}
  }
  BP_SidebarCollectionScreenContainerTabButtonEditTags {
   widgetClassPath=&/BP_SidebarCollectionScreenContainerTabButtonEditTags
   bIsVariable=true
   parent=ScaleBox_TagCategoryIcon
   widget=BP_SidebarCollectionScreenContainerTabButtonEditTags
   properties {
    onButtonBaseClicked=(&self:WidgetTree.BP_SidebarCollectionScreenContainerTabButtonEditTags.HandleOnButtonClicked)
    onButtonBaseDoubleClicked=(&self:WidgetTree.BP_SidebarCollectionScreenContainerTabButtonEditTags.HandleOnButtonDoubleClicked)
    onButtonBaseHovered=(&self:WidgetTree.BP_SidebarCollectionScreenContainerTabButtonEditTags.HandleOnButtonHovered)
    onButtonBaseUnhovered=(&self:WidgetTree.BP_SidebarCollectionScreenContainerTabButtonEditTags.HandleOnButtonUnhovered)
    onSelectedChangedBase=(&self:WidgetTree.BP_SidebarCollectionScreenContainerTabButtonEditTags.HandleOnSelectedChanged)
    visibility=HitTestInvisible
   }
   slot {class=ScaleBoxSlot}
  }
  Border_TagBorder {
   widgetClassPath=&Border
   bIsVariable=true
   parent=GridPanel_TagWidget
   widget=Border_TagBorder
   properties {
    background {
     bIsDynamicallyLoaded=false
     drawAs=Image
     imageType=NoImage
     mirroring=NoMirror
     resourceName=None
     resourceObject=/Game/UI/Sidebar/Tags/Materials/BoxSkew_SocialTag
     tiling=NoTile
     imageSize=<32 32>
     margin=<ltrb 0 0 0 0>
     outlineSettings {
      bUseBrushTransparency=false
      roundingType=HalfHeightRadius
      width=0
      color {colorUseRule=UseColor_Specified, specifiedColor=<rgba 0 0 0 0>}
      cornerRadii=<0 0 0 0>
     }
     tintColor {colorUseRule=UseColor_Specified, specifiedColor=<rgba 1 1 1 1>}
     uVRegion {bIsValid=false, max=<0 0>, min=<0 0>}
    }
    padding=<ltrb 8 6 7 4>
   }
   slot {class=GridSlot, properties {column=1, verticalAlignment=VAlign_Center}}
  }
  HorizontalBox_55 {
   widgetClassPath=&HorizontalBox
   parent=Border_TagBorder
   widget=HorizontalBox_55
   slot {class=BorderSlot, properties {padding=<ltrb 8 6 7 4>}}
  }
  Spacer_TagCategoryIcon {
   widgetClassPath=&Spacer
   bIsVariable=true
   parent=HorizontalBox_55
   widget=Spacer_TagCategoryIcon
   properties {size=<18 1>}
   slot {class=HorizontalBoxSlot}
  }
  Text_TagName {
   widgetClassPath=&CommonTextBlock
   bIsVariable=true
   parent=HorizontalBox_55
   widget=Text_TagName
   properties {
    bIsScrollingEnabled=false
    clipping=ClipToBounds
    mobileFontSizeMultiplier=1.25
    scrollStyle=&TextScrollStyle-Base_C
    text=Good Vibes
    textOverflowPolicy=Ellipsis
    colorAndOpacity {colorUseRule=UseColor_Specified, specifiedColor=<rgba 1 1 1 0.902>}
    font {
     bForceMonospaced=false
     bMaterialIsStencil=false
     fontMaterial=null
     fontObject=&BurbankSmall
     letterSpacing=0
     monospacedWidth=1
     size=16
     skewAmount=0.15
     typefaceFontName=Black
     outlineSettings {
      bApplyOutlineToDropShadows=false
      bMiteredCorners=false
      bSeparateFillAlpha=false
      outlineMaterial=null
      outlineSize=0
      outlineColor=<rgba 0 0 0 1>
     }
    }
    margin=<ltrb 0 0 10 0>
    shadowOffset=<0 0>
   }
   slot {class=HorizontalBoxSlot, properties {verticalAlignment=VAlign_Center, padding=<ltrb 5 0 5 0>, size {sizeRule=Fill, value=1}}}
  }
  Image_CheckMark {
   widgetClassPath=&Image
   bInherited=true
   bIsVariable=true
   parent=HorizontalBox_55
   widget=Image_CheckMark
   properties {
    visibility=SelfHitTestInvisible
    brush {
     bIsDynamicallyLoaded=false
     drawAs=Image
     imageType=NoImage
     mirroring=NoMirror
     resourceName=None
     resourceObject=/Game/UI/Foundation/Textures/Icons/Checks/T-Icon-Check-64
     tiling=NoTile
     imageSize=<24 24>
     margin=<ltrb 0 0 0 0>
     outlineSettings {
      bUseBrushTransparency=false
      roundingType=HalfHeightRadius
      width=0
      color {colorUseRule=UseColor_Specified, specifiedColor=<rgba 0 0 0 0>}
      cornerRadii=<0 0 0 0>
     }
     tintColor {colorUseRule=UseColor_Specified, specifiedColor=<rgba 1 1 1 1>}
     uVRegion {bIsValid=false, max=<0 0>, min=<0 0>}
    }
   }
   slot {class=HorizontalBoxSlot, properties {horizontalAlignment=HAlign_Right, verticalAlignment=VAlign_Center, padding=<ltrb 0 0 5 2>}}
  }
  SizeBox_134 {
   widgetClassPath=&SizeBox
   parent=HorizontalBox_55
   widget=SizeBox_134
   properties {bOverride_HeightOverride=true, bOverride_WidthOverride=true, heightOverride=26, widthOverride=26}
   slot {class=HorizontalBoxSlot, properties {horizontalAlignment=HAlign_Right, verticalAlignment=VAlign_Center, padding=<ltrb 0 0 5 0>}}
  }
  InputActionWidget {
   widgetClassPath=&CommonActionWidget
   bInherited=true
   parent=SizeBox_134
   widget=InputActionWidget
   properties {
    visibility=SelfHitTestInvisible
    inputActions [{dataTable=&InputActionDataTable, rowName=Select}]
    progressMaterialBrush {
     bIsDynamicallyLoaded=false
     drawAs=Image
     imageType=NoImage
     mirroring=NoMirror
     resourceName=None
     resourceObject=null
     tiling=NoTile
     imageSize=<16 16>
     margin=<ltrb 0 0 0 0>
     outlineSettings {
      bUseBrushTransparency=false
      roundingType=HalfHeightRadius
      width=0
      color {colorUseRule=UseColor_Specified, specifiedColor=<rgba 0 0 0 0>}
      cornerRadii=<0 0 0 0>
     }
     tintColor {colorUseRule=UseColor_Specified, specifiedColor=<rgba 1 1 1 1>}
     uVRegion {bIsValid=false, max=<0 0>, min=<0 0>}
    }
   }
   slot {class=SizeBoxSlot}
  }
  SidebarSocialAddingTagWidget {
   widgetClassPath=&/SidebarSocialAddingTagWidget
   bIsVariable=true
   parent=Switcher_SelectedState
   widget=SidebarSocialAddingTagWidget
   slot {class=CommonVisibilitySwitcherSlot, properties {horizontalAlignment=HAlign_Left, verticalAlignment=VAlign_Center}}
  }
 ]
}
